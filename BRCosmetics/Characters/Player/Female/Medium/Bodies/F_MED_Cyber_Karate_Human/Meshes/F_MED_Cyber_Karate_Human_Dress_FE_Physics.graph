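# Physics asset for the female Cyber Karate Human dress, rigging its skirt and back collar for physical motion.
# Twelve skirt chains each link three box bodies with limited swing and tight twist below a small MSTR capsule.
# Left and right front chains anchor to the thighs, others to the pelvis, and two collar flaps to spine_05 through master capsules.
# Anchor and body-capsule constraints rotate freely; only the left thigh has a twist capsule.
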PhysicsAsset /BRCosmetics/Characters/Player/Female/Medium/Bodies/F_MED_Cyber_Karate_Human/Meshes/F_MED_Cyber_Karate_Human_Dress_FE_Physics
bodies {
 dyn_collar_back_l [{shapeType=Box, extentX=9.3795, extentY=1.962, extentZ=26.2815, center=<4.7592 0 0>}]
 dyn_collar_back_r [{shapeType=Box, extentX=9.6075, extentY=1.962, extentZ=27.1684, center=<4.8188 0 0>}]
 dyn_skirt_01_l [{shapeType=Box, extentX=10.2693, extentY=13.2247, extentZ=1.7471, center=<4.9784 0 0>}]
 dyn_skirt_01_l_MSTR [{shapeType=Capsule, length=1, radius=1}]
 dyn_skirt_01_r [{shapeType=Box, extentX=10.2693, extentY=13.2247, extentZ=1.7471, center=<4.9784 0 0>}]
 dyn_skirt_01_r_MSTR [{shapeType=Capsule, length=1, radius=1}]
 dyn_skirt_02_l [{shapeType=Box, extentX=10.2693, extentY=13.2247, extentZ=1.7471, center=<4.9784 0 0>}]
 dyn_skirt_02_r [{shapeType=Box, extentX=10.2693, extentY=13.2247, extentZ=1.7471, center=<4.9784 0 0>}]
 dyn_skirt_03_l [{shapeType=Box, extentX=10.2693, extentY=13.2247, extentZ=1.7471, center=<4.9784 0 0>}]
 dyn_skirt_03_r [{shapeType=Box, extentX=10.2693, extentY=13.2247, extentZ=1.7471, center=<4.9784 0 0>}]
 dyn_skirt_bk2_01_l [{shapeType=Box, extentX=10.2693, extentY=13.2247, extentZ=1.7471, center=<4.9784 0 0>}]
 dyn_skirt_bk2_01_l_MSTR [{shapeType=Capsule, length=1, radius=1}]
 dyn_skirt_bk2_01_r [{shapeType=Box, extentX=10.2693, extentY=13.2247, extentZ=1.7471, center=<4.9784 0 0>}]
 dyn_skirt_bk2_01_r_MSTR [{shapeType=Capsule, length=1, radius=1}]
 dyn_skirt_bk2_02_l [{shapeType=Box, extentX=10.2693, extentY=13.2247, extentZ=1.7471, center=<4.9784 0 0>}]
 dyn_skirt_bk2_02_r [{shapeType=Box, extentX=10.2693, extentY=13.2247, extentZ=1.7471, center=<4.9784 0 0>}]
 dyn_skirt_bk2_03_l [{shapeType=Box, extentX=10.2693, extentY=13.2247, extentZ=1.7471, center=<4.9784 0 0>}]
 dyn_skirt_bk2_03_r [{shapeType=Box, extentX=10.2693, extentY=13.2247, extentZ=1.7471, center=<4.9784 0 0>}]
 dyn_skirt_bk_01 [{shapeType=Box, extentX=10.2693, extentY=13.2247, extentZ=1.7471, center=<4.9784 0 0>}]
 dyn_skirt_bk_01_MSTR [{shapeType=Capsule, length=1, radius=1}]
 dyn_skirt_bk_01_l [{shapeType=Box, extentX=10.2693, extentY=13.2247, extentZ=1.7471, center=<4.9784 0 0>}]
 dyn_skirt_bk_01_l_MSTR [{shapeType=Capsule, length=1, radius=1}]
 dyn_skirt_bk_01_r [{shapeType=Box, extentX=10.2693, extentY=13.2247, extentZ=1.7471, center=<4.9784 0 0>}]
 dyn_skirt_bk_01_r_MSTR [{shapeType=Capsule, length=1, radius=1}]
 dyn_skirt_bk_02 [{shapeType=Box, extentX=10.2693, extentY=13.2247, extentZ=1.7471, center=<4.9784 0 0>}]
 dyn_skirt_bk_02_l [{shapeType=Box, extentX=10.2693, extentY=13.2247, extentZ=1.7471, center=<4.9784 0 0>}]
 dyn_skirt_bk_02_r [{shapeType=Box, extentX=10.2693, extentY=13.2247, extentZ=1.7471, center=<4.9784 0 0>}]
 dyn_skirt_bk_03 [{shapeType=Box, extentX=10.2693, extentY=13.2247, extentZ=1.7471, center=<4.9784 0 0>}]
 dyn_skirt_bk_03_l [{shapeType=Box, extentX=10.2693, extentY=13.2247, extentZ=1.7471, center=<4.9784 0 0>}]
 dyn_skirt_bk_03_r [{shapeType=Box, extentX=10.2693, extentY=13.2247, extentZ=1.7471, center=<4.9784 0 0>}]
 dyn_skirt_fr2_01_l [{shapeType=Box, extentX=10.2693, extentY=13.2247, extentZ=1.7471, center=<4.9784 0 0>}]
 dyn_skirt_fr2_01_l_MSTR [{shapeType=Capsule, length=1, radius=1}]
 dyn_skirt_fr2_01_r [{shapeType=Box, extentX=10.2693, extentY=13.2247, extentZ=1.7471, center=<4.9784 0 0>}]
 dyn_skirt_fr2_01_r_MSTR [{shapeType=Capsule, length=1, radius=1}]
 dyn_skirt_fr2_02_l [{shapeType=Box, extentX=10.2693, extentY=13.2247, extentZ=1.7471, center=<4.9784 0 0>}]
 dyn_skirt_fr2_02_r [{shapeType=Box, extentX=10.2693, extentY=13.2247, extentZ=1.7471, center=<4.9784 0 0>}]
 dyn_skirt_fr2_03_l [{shapeType=Box, extentX=10.2693, extentY=13.2247, extentZ=1.7471, center=<4.9784 0 0>}]
 dyn_skirt_fr2_03_r [{shapeType=Box, extentX=10.2693, extentY=13.2247, extentZ=1.7471, center=<4.9784 0 0>}]
 dyn_skirt_fr_01 [{shapeType=Box, extentX=10.2693, extentY=13.2247, extentZ=2.2721, center=<4.9784 0 0>}]
 dyn_skirt_fr_01_MSTR [{shapeType=Capsule, length=1, radius=1}]
 dyn_skirt_fr_01_l [{shapeType=Box, extentX=10.2693, extentY=13.2247, extentZ=2.6286, center=<4.9784 0 0>}]
 dyn_skirt_fr_01_l_MSTR [{shapeType=Capsule, length=1, radius=1}]
 dyn_skirt_fr_01_r [{shapeType=Box, extentX=10.2693, extentY=13.2247, extentZ=1.7471, center=<4.9784 0 0>}]
 dyn_skirt_fr_01_r_MSTR [{shapeType=Capsule, length=1, radius=1}]
 dyn_skirt_fr_02 [{shapeType=Box, extentX=10.2693, extentY=13.2247, extentZ=2.821, center=<4.9784 0 0>}]
 dyn_skirt_fr_02_l [{shapeType=Box, extentX=10.2693, extentY=13.2247, extentZ=2.7756, center=<4.9784 0 0>}]
 dyn_skirt_fr_02_r [{shapeType=Box, extentX=10.2693, extentY=13.2247, extentZ=1.7471, center=<4.9784 0 0>}]
 dyn_skirt_fr_03 [{shapeType=Box, extentX=10.2693, extentY=13.2247, extentZ=2.821, center=<4.9784 0 0>}]
 dyn_skirt_fr_03_l [{shapeType=Box, extentX=10.2693, extentY=13.2247, extentZ=2.6658, center=<4.9784 0 0>}]
 dyn_skirt_fr_03_r [{shapeType=Box, extentX=10.2693, extentY=13.2247, extentZ=1.7471, center=<4.9784 0 0>}]
 master_collar_back_l [{shapeType=Capsule, length=1, radius=1}]
 master_collar_back_r [{shapeType=Capsule, length=1, radius=1}]
 pelvis [{shapeType=Capsule, length=8.7692, radius=6.7524, center=<1.1549 -1.274 0>}]
 spine_04 [{shapeType=Capsule, length=51.1536, radius=6.9446}]
 spine_05 [{shapeType=Capsule, length=47.212, radius=6.1624}]
 thigh_l [{shapeType=Capsule, length=29.6749, radius=7.6572, center=<-21.4384 -0.0002 0>, rotation=<rot 85.5494 -180 -180>}]
 thigh_r [{shapeType=Capsule, length=30.0985, radius=7.2337, center=<21.3462 -0.0004 -0.3121>, rotation=<rot 85.8088 179.903 -179.9999>}]
 thigh_twist_01_l [{shapeType=Capsule, length=29.6749, radius=7.6572, center=<6.3077 -0.0002 0.2061>, rotation=<rot 85.5298 -180 -180>}]
}
constraints [
 {
  bone1Name=thigh_l
  bone2Name=pelvis
  swing1LimitDegrees=45
  swing1Motion=Free
  swing2LimitDegrees=45
  swing2Motion=Free
  twistLimitDegrees=45
  twistMotion=Free
 }
 {
  bone1Name=thigh_r
  bone2Name=pelvis
  swing1LimitDegrees=45
  swing1Motion=Free
  swing2LimitDegrees=45
  swing2Motion=Free
  twistLimitDegrees=45
  twistMotion=Free
 }
 {
  bone1Name=dyn_skirt_fr_01_l_MSTR
  bone2Name=thigh_l
  swing1LimitDegrees=45
  swing1Motion=Free
  swing2LimitDegrees=45
  swing2Motion=Free
  twistLimitDegrees=45
  twistMotion=Free
 }
 {
  bone1Name=dyn_skirt_fr2_01_l_MSTR
  bone2Name=thigh_l
  swing1LimitDegrees=45
  swing1Motion=Free
  swing2LimitDegrees=45
  swing2Motion=Free
  twistLimitDegrees=45
  twistMotion=Free
 }
 {
  bone1Name=dyn_skirt_fr_01_r_MSTR
  bone2Name=thigh_r
  swing1LimitDegrees=45
  swing1Motion=Free
  swing2LimitDegrees=45
  swing2Motion=Free
  twistLimitDegrees=45
  twistMotion=Free
 }
 {
  bone1Name=dyn_skirt_fr2_01_r_MSTR
  bone2Name=thigh_r
  swing1LimitDegrees=45
  swing1Motion=Free
  swing2LimitDegrees=45
  swing2Motion=Free
  twistLimitDegrees=45
  twistMotion=Free
 }
 {
  bone1Name=dyn_skirt_fr_01_MSTR
  bone2Name=pelvis
  swing1LimitDegrees=45
  swing1Motion=Free
  swing2LimitDegrees=45
  swing2Motion=Free
  twistLimitDegrees=45
  twistMotion=Free
 }
 {
  bone1Name=dyn_skirt_bk_01_MSTR
  bone2Name=pelvis
  swing1LimitDegrees=45
  swing1Motion=Free
  swing2LimitDegrees=45
  swing2Motion=Free
  twistLimitDegrees=45
  twistMotion=Free
 }
 {
  bone1Name=dyn_skirt_01_r_MSTR
  bone2Name=pelvis
  swing1LimitDegrees=45
  swing1Motion=Free
  swing2LimitDegrees=45
  swing2Motion=Free
  twistLimitDegrees=45
  twistMotion=Free
 }
 {
  bone1Name=dyn_skirt_bk2_01_r_MSTR
  bone2Name=pelvis
  swing1LimitDegrees=45
  swing1Motion=Free
  swing2LimitDegrees=45
  swing2Motion=Free
  twistLimitDegrees=45
  twistMotion=Free
 }
 {
  bone1Name=dyn_skirt_bk_01_r_MSTR
  bone2Name=pelvis
  swing1LimitDegrees=45
  swing1Motion=Free
  swing2LimitDegrees=45
  swing2Motion=Free
  twistLimitDegrees=45
  twistMotion=Free
 }
 {
  bone1Name=dyn_skirt_bk_01_l_MSTR
  bone2Name=pelvis
  swing1LimitDegrees=45
  swing1Motion=Free
  swing2LimitDegrees=45
  swing2Motion=Free
  twistLimitDegrees=45
  twistMotion=Free
 }
 {
  bone1Name=dyn_skirt_01_l_MSTR
  bone2Name=pelvis
  swing1LimitDegrees=45
  swing1Motion=Free
  swing2LimitDegrees=45
  swing2Motion=Free
  twistLimitDegrees=45
  twistMotion=Free
 }
 {
  bone1Name=dyn_skirt_bk2_01_l_MSTR
  bone2Name=pelvis
  swing1LimitDegrees=45
  swing1Motion=Free
  swing2LimitDegrees=45
  swing2Motion=Free
  twistLimitDegrees=45
  twistMotion=Free
 }
 {bone1Name=dyn_skirt_03_l, bone2Name=dyn_skirt_02_l, swing1LimitDegrees=25, swing2LimitDegrees=45, twistLimitDegrees=5}
 {bone1Name=dyn_skirt_01_l, bone2Name=dyn_skirt_01_l_MSTR, swing1LimitDegrees=25, swing2LimitDegrees=45, twistLimitDegrees=5}
 {bone1Name=dyn_skirt_02_l, bone2Name=dyn_skirt_01_l, swing1LimitDegrees=25, swing2LimitDegrees=45, twistLimitDegrees=5}
 {bone1Name=dyn_skirt_bk2_03_l, bone2Name=dyn_skirt_bk2_02_l, swing1LimitDegrees=25, swing2LimitDegrees=45, twistLimitDegrees=5}
 {bone1Name=dyn_skirt_bk2_01_l, bone2Name=dyn_skirt_bk2_01_l_MSTR, swing1LimitDegrees=25, swing2LimitDegrees=45, twistLimitDegrees=5}
 {bone1Name=dyn_skirt_bk2_02_l, bone2Name=dyn_skirt_bk2_01_l, swing1LimitDegrees=25, swing2LimitDegrees=45, twistLimitDegrees=5}
 {bone1Name=dyn_skirt_bk_03_l, bone2Name=dyn_skirt_bk_02_l, swing1LimitDegrees=25, swing2LimitDegrees=45, twistLimitDegrees=5}
 {bone1Name=dyn_skirt_bk_01_l, bone2Name=dyn_skirt_bk_01_l_MSTR, swing1LimitDegrees=25, swing2LimitDegrees=45, twistLimitDegrees=5}
 {bone1Name=dyn_skirt_bk_02_l, bone2Name=dyn_skirt_bk_01_l, swing1LimitDegrees=25, swing2LimitDegrees=45, twistLimitDegrees=5}
 {bone1Name=dyn_skirt_bk_03_r, bone2Name=dyn_skirt_bk_02_r, swing1LimitDegrees=25, swing2LimitDegrees=45, twistLimitDegrees=5}
 {bone1Name=dyn_skirt_bk_01_r, bone2Name=dyn_skirt_bk_01_r_MSTR, swing1LimitDegrees=25, swing2LimitDegrees=45, twistLimitDegrees=5}
 {bone1Name=dyn_skirt_bk_02_r, bone2Name=dyn_skirt_bk_01_r, swing1LimitDegrees=25, swing2LimitDegrees=45, twistLimitDegrees=5}
 {bone1Name=dyn_skirt_bk2_03_r, bone2Name=dyn_skirt_bk2_02_r, swing1LimitDegrees=25, swing2LimitDegrees=45, twistLimitDegrees=5}
 {bone1Name=dyn_skirt_bk2_01_r, bone2Name=dyn_skirt_bk2_01_r_MSTR, swing1LimitDegrees=25, swing2LimitDegrees=45, twistLimitDegrees=5}
 {bone1Name=dyn_skirt_bk2_02_r, bone2Name=dyn_skirt_bk2_01_r, swing1LimitDegrees=25, swing2LimitDegrees=45, twistLimitDegrees=5}
 {bone1Name=dyn_skirt_03_r, bone2Name=dyn_skirt_02_r, swing1LimitDegrees=25, swing2LimitDegrees=45, twistLimitDegrees=5}
 {bone1Name=dyn_skirt_01_r, bone2Name=dyn_skirt_01_r_MSTR, swing1LimitDegrees=25, swing2LimitDegrees=45, twistLimitDegrees=5}
 {bone1Name=dyn_skirt_02_r, bone2Name=dyn_skirt_01_r, swing1LimitDegrees=25, swing2LimitDegrees=45, twistLimitDegrees=5}
 {bone1Name=dyn_skirt_bk_03, bone2Name=dyn_skirt_bk_02, swing1LimitDegrees=25, swing2LimitDegrees=45, twistLimitDegrees=5}
 {bone1Name=dyn_skirt_bk_01, bone2Name=dyn_skirt_bk_01_MSTR, swing1LimitDegrees=25, swing2LimitDegrees=45, twistLimitDegrees=5}
 {bone1Name=dyn_skirt_bk_02, bone2Name=dyn_skirt_bk_01, swing1LimitDegrees=25, swing2LimitDegrees=45, twistLimitDegrees=5}
 {bone1Name=dyn_skirt_fr_03, bone2Name=dyn_skirt_fr_02, swing1LimitDegrees=25, swing2LimitDegrees=45, twistLimitDegrees=5}
 {bone1Name=dyn_skirt_fr_01, bone2Name=dyn_skirt_fr_01_MSTR, swing1LimitDegrees=25, swing2LimitDegrees=45, twistLimitDegrees=5}
 {bone1Name=dyn_skirt_fr_02, bone2Name=dyn_skirt_fr_01, swing1LimitDegrees=25, swing2LimitDegrees=45, twistLimitDegrees=5}
 {bone1Name=dyn_skirt_fr2_03_r, bone2Name=dyn_skirt_fr2_02_r, swing1LimitDegrees=25, swing2LimitDegrees=45, twistLimitDegrees=5}
 {bone1Name=dyn_skirt_fr2_01_r, bone2Name=dyn_skirt_fr2_01_r_MSTR, swing1LimitDegrees=25, swing2LimitDegrees=45, twistLimitDegrees=5}
 {bone1Name=dyn_skirt_fr2_02_r, bone2Name=dyn_skirt_fr2_01_r, swing1LimitDegrees=25, swing2LimitDegrees=45, twistLimitDegrees=5}
 {bone1Name=dyn_skirt_fr_03_r, bone2Name=dyn_skirt_fr_02_r, swing1LimitDegrees=25, swing2LimitDegrees=45, twistLimitDegrees=5}
 {bone1Name=dyn_skirt_fr_01_r, bone2Name=dyn_skirt_fr_01_r_MSTR, swing1LimitDegrees=25, swing2LimitDegrees=45, twistLimitDegrees=5}
 {bone1Name=dyn_skirt_fr_02_r, bone2Name=dyn_skirt_fr_01_r, swing1LimitDegrees=25, swing2LimitDegrees=45, twistLimitDegrees=5}
 {bone1Name=dyn_skirt_fr2_03_l, bone2Name=dyn_skirt_fr2_02_l, swing1LimitDegrees=25, swing2LimitDegrees=45, twistLimitDegrees=5}
 {bone1Name=dyn_skirt_fr2_01_l, bone2Name=dyn_skirt_fr2_01_l_MSTR, swing1LimitDegrees=25, swing2LimitDegrees=45, twistLimitDegrees=5}
 {bone1Name=dyn_skirt_fr2_02_l, bone2Name=dyn_skirt_fr2_01_l, swing1LimitDegrees=25, swing2LimitDegrees=45, twistLimitDegrees=5}
 {bone1Name=dyn_skirt_fr_03_l, bone2Name=dyn_skirt_fr_02_l, swing1LimitDegrees=25, swing2LimitDegrees=45, twistLimitDegrees=5}
 {bone1Name=dyn_skirt_fr_01_l, bone2Name=dyn_skirt_fr_01_l_MSTR, swing1LimitDegrees=25, swing2LimitDegrees=45, twistLimitDegrees=5}
 {bone1Name=dyn_skirt_fr_02_l, bone2Name=dyn_skirt_fr_01_l, swing1LimitDegrees=25, swing2LimitDegrees=45, twistLimitDegrees=5}
 {
  bone1Name=master_collar_back_l
  bone2Name=spine_05
  swing1LimitDegrees=45
  swing1Motion=Free
  swing2LimitDegrees=45
  swing2Motion=Free
  twistLimitDegrees=45
  twistMotion=Free
 }
 {
  bone1Name=master_collar_back_r
  bone2Name=spine_05
  swing1LimitDegrees=45
  swing1Motion=Free
  swing2LimitDegrees=45
  swing2Motion=Free
  twistLimitDegrees=45
  twistMotion=Free
 }
 {bone1Name=dyn_collar_back_l, bone2Name=master_collar_back_l, swing1LimitDegrees=45, swing2LimitDegrees=45, twistLimitDegrees=45}
 {bone1Name=dyn_collar_back_r, bone2Name=master_collar_back_r, swing1LimitDegrees=45, swing2LimitDegrees=45, twistLimitDegrees=45}
 {
  bone1Name=spine_05
  bone2Name=spine_04
  swing1LimitDegrees=45
  swing1Motion=Free
  swing2LimitDegrees=45
  swing2Motion=Free
  twistLimitDegrees=45
  twistMotion=Free
 }
 {
  bone1Name=spine_04
  bone2Name=pelvis
  swing1LimitDegrees=45
  swing1Motion=Free
  swing2LimitDegrees=45
  swing2Motion=Free
  twistLimitDegrees=45
  twistMotion=Free
 }
 {
  bone1Name=thigh_twist_01_l
  bone2Name=thigh_l
  swing1LimitDegrees=45
  swing1Motion=Free
  swing2LimitDegrees=45
  swing2Motion=Free
  twistLimitDegrees=45
  twistMotion=Free
 }
]
