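# Physics asset for the F_MED_Cyber_Karate_Human body's scarf, providing capsule bodies and joints for its dangling tie and armband bones.
# A four-link tie chain hangs from spine_05 with tight twist and wider, per-link swing limits.
# The armband's master knot joins spine_05 with free rotation, then a limited knot forks into high and low two-bone tails.
# Large spine_03 and thigh capsules supply body volumes, with the thighs left unconstrained.
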PhysicsAsset /BRCosmetics/Characters/Player/Female/Medium/Bodies/F_MED_Cyber_Karate_Human/Meshes/F_MED_Cyber_Karate_Human_Scarf_Physics
bodies {
 dyn_armband_hi_01 [{shapeType=Capsule, length=2, radius=1, center=<0.7817 0.0056 0.1256>, rotation=<rot 81.3574 162.9993 -179.9999>}]
 dyn_armband_hi_02 [{shapeType=Capsule, length=2, radius=1, center=<0.7329 0.0206 -0.1928>, rotation=<rot 69.9999 -17.0002 0>}]
 dyn_armband_knot [{shapeType=Capsule, length=2, radius=1, rotation=<rot 85.1685 179.9994 179.9994>}]
 dyn_armband_knot_MSTR [{shapeType=Capsule, length=0.505, radius=0.505}]
 dyn_armband_low_01 [{shapeType=Capsule, length=2, radius=1, center=<1.0021 0 -0.4403>, rotation=<rot 66.2815 0 0>}]
 dyn_armband_low_02 [{shapeType=Capsule, length=2, radius=1, center=<0.3373 0.1415 0.1761>, rotation=<rot 88.5221 -17.0008 0.0004>}]
 dyn_tie_01 [{shapeType=Capsule, length=2, radius=2, center=<2.6116 0 0>, rotation=<rot 90 0 0>}]
 dyn_tie_02 [{shapeType=Capsule, length=1.7995, radius=1.0413, center=<2.9376 0 0>, rotation=<rot 90 0 0>}]
 dyn_tie_03 [
  {shapeType=Capsule, length=1.1007, radius=2.8993, center=<2.645 0 0>, rotation=<rot 90 0 0>}
  {shapeType=Capsule, length=5.9579, radius=2.6057, center=<5.4934 -2.1747 -0.698>, rotation=<rot 90 0 0>}
 ]
 dyn_tie_04 [{shapeType=Capsule, length=1.9171, radius=2.8588, center=<5.425 0.0001 0>, rotation=<rot 90 0 0>}]
 spine_03 [
  {shapeType=Capsule, length=14.75, radius=6.75, center=<5.9545 5.6792 -0.0001>}
  {shapeType=Capsule, length=12.0192, radius=9.9378, center=<-7.2014 0.3423 -0.518>, rotation=<rot -2.8498 -84.6116 86.722>}
 ]
 spine_05 [{shapeType=Capsule, length=2, radius=2}]
 thigh_l [{
   shapeType=Capsule
   length=24.5
   radius=13.25
   shapeName=thigh_l_capsule
   center=<-27.883 0.0007 0.1927>
   rotation=<rot -89.6019 179.9912 -179.9912>
  }]
 thigh_r [{
   shapeType=Capsule
   length=24.5
   radius=13.25
   shapeName=thigh_r_capsule
   center=<27.871 0 -0.1954>
   rotation=<rot -89.6004 179.9912 -179.9912>
  }]
}
constraints [
 {
  bone1Name=spine_05
  bone2Name=spine_03
  swing1LimitDegrees=45
  swing1Motion=Free
  swing2LimitDegrees=45
  swing2Motion=Free
  twistLimitDegrees=45
  twistMotion=Free
 }
 {bone1Name=dyn_tie_01, bone2Name=spine_05, swing1LimitDegrees=20, swing2LimitDegrees=45, twistLimitDegrees=3}
 {bone1Name=dyn_tie_02, bone2Name=dyn_tie_01, swing1LimitDegrees=20, swing2LimitDegrees=40, twistLimitDegrees=3}
 {bone1Name=dyn_tie_03, bone2Name=dyn_tie_02, swing1LimitDegrees=25, swing2LimitDegrees=40, twistLimitDegrees=3}
 {bone1Name=dyn_tie_04, bone2Name=dyn_tie_03, swing1LimitDegrees=35, swing2LimitDegrees=30, twistLimitDegrees=3}
 {
  bone1Name=dyn_armband_knot_MSTR
  bone2Name=spine_05
  swing1LimitDegrees=45
  swing1Motion=Free
  swing2LimitDegrees=45
  swing2Motion=Free
  twistLimitDegrees=45
  twistMotion=Free
 }
 {bone1Name=dyn_armband_knot, bone2Name=dyn_armband_knot_MSTR, swing1LimitDegrees=20, swing2LimitDegrees=20, twistLimitDegrees=3}
 {bone1Name=dyn_armband_low_01, bone2Name=dyn_armband_knot, swing1LimitDegrees=20, swing2LimitDegrees=20, twistLimitDegrees=3}
 {bone1Name=dyn_armband_low_02, bone2Name=dyn_armband_low_01, swing1LimitDegrees=20, swing2LimitDegrees=20, twistLimitDegrees=3}
 {bone1Name=dyn_armband_hi_01, bone2Name=dyn_armband_knot, swing1LimitDegrees=20, swing2LimitDegrees=20, twistLimitDegrees=3}
 {bone1Name=dyn_armband_hi_02, bone2Name=dyn_armband_hi_01, swing1LimitDegrees=20, swing2LimitDegrees=20, twistLimitDegrees=3}
]
