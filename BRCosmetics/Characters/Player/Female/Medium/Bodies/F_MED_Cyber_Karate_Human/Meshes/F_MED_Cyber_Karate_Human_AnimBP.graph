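# Animation Blueprint for the F_MED_Cyber_Karate_Human body that drives procedural motion of its skirt bones.
# From a cached pre-dress pose, a skydiving flag selects back-skirt constraints plus AnimDynamics when true, or a pelvis-space RigidBody simulation when false.
# Shared constraints then pull front bones toward the thighs and average neighboring bones to spread motion around the skirt.
# Spring-based AnimDynamics on front bones follows, using the variable skirt gravity override.
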AnimBlueprint /BRCosmetics/Characters/Player/Female/Medium/Bodies/F_MED_Cyber_Karate_Human/Meshes/F_MED_Cyber_Karate_Human_AnimBP
target_skeleton=&/F_MED_Cyber_Karate_Human_Skeleton
graphs [
 AnimGraph {nodes [
   RigidBody!(ComponentPose, Alpha=0.5, OverrideWorldGravity:Vector=Get Gravity_Override_Skirt() -> (Gravity_Override_Skirt:Vector), ExternalForce) -> (Pose) {
    bClampLinearTranslationLimitToRefPose=true
    bOverrideWorldGravity=true
    overridePhysicsAsset=&/F_MED_Cyber_Karate_Human_Dress_Physics
    simulationSpace=BaseBoneSpace
    alphaBoolBlend {blendOption=Linear}
    baseBoneRef {boneName=root}
    componentAppliedLinearAccClamp=<3000 3000 1000>
    componentLinearAccScale=<0.35 0.35 0.4>
    overrideWorldGravity {z=-980}
   }
   n1 = "Constraint - dyn_skirt_bk_01 (RPP)"!(
    ComponentPose=Local To Component(LocalPose=Use cached pose 'PreDressRBAN'())
    Alpha=1
    ConstraintWeights_0=0
    ConstraintWeights_1:float=0.5
    ConstraintWeights_2:float=0.5
   ) -> (Pose) {
    alphaBoolBlend {blendOption=Linear}
    boneToModify {boneName=dyn_skirt_bk_01}
    constraintSetup [{transformType=Rotation, targetBone {boneName=pelvis}}, {targetBone {boneName=dyn_skirt_bk_01_l}}, {targetBone {boneName=dyn_skirt_bk_01_r}}]
    constraintWeights [1,1,1]
   }
   n2 = "Constraint - dyn_skirt_bk_02 (RPP)"!(ComponentPose=@n1, Alpha=1, ConstraintWeights_0=0, ConstraintWeights_1:float=0.5, ConstraintWeights_2:float=0.5) -> (Pose) {
    alphaBoolBlend {blendOption=Linear}
    boneToModify {boneName=dyn_skirt_bk_02}
    constraintSetup [{transformType=Rotation, targetBone {boneName=pelvis}}, {targetBone {boneName=dyn_skirt_bk_02_l}}, {targetBone {boneName=dyn_skirt_bk_02_r}}]
    constraintWeights [1,1,1]
   }
   n3 = "Constraint - dyn_skirt_bk_03 (RPP)"!(ComponentPose=@n2, Alpha=1, ConstraintWeights_0=0, ConstraintWeights_1:float=0.5, ConstraintWeights_2:float=0.5) -> (Pose) {
    alphaBoolBlend {blendOption=Linear}
    boneToModify {boneName=dyn_skirt_bk_03}
    constraintSetup [{transformType=Rotation, targetBone {boneName=pelvis}}, {targetBone {boneName=dyn_skirt_bk_03_l}}, {targetBone {boneName=dyn_skirt_bk_03_r}}]
    constraintWeights [1,1,1]
   }
   n4 = "Constraint - dyn_skirt_bk2_01_l (P)"(ComponentPose=@n3, Alpha=1, ConstraintWeights_0=1) {
    alphaBoolBlend {blendOption=Linear}
    boneToModify {boneName=dyn_skirt_bk2_01_l}
    constraintSetup [{targetBone {boneName=pelvis}}]
    constraintWeights [1]
   }
   n5 = "Constraint - dyn_skirt_bk2_02_l (P)"(ComponentPose=@n4, Alpha=0.3, ConstraintWeights_0=1) {
    alphaBoolBlend {blendOption=Linear}
    boneToModify {boneName=dyn_skirt_bk2_02_l}
    constraintSetup [{targetBone {boneName=pelvis}}]
    constraintWeights [1]
   }
   n6 = "Constraint - dyn_skirt_bk2_03_l (P)"(ComponentPose=@n5, Alpha=1, ConstraintWeights_0=1) {
    alphaBoolBlend {blendOption=Linear}
    boneToModify {boneName=dyn_skirt_bk2_03_l}
    constraintSetup [{targetBone {boneName=pelvis}}]
    constraintWeights [1]
   }
   n7 = "Constraint - dyn_skirt_bk_01_l (RRR)"!(ComponentPose=@n6, Alpha=1, ConstraintWeights_0=0, ConstraintWeights_1:float=0.5, ConstraintWeights_2:float=0.5) -> (Pose) {
    alphaBoolBlend {blendOption=Linear}
    boneToModify {boneName=dyn_skirt_bk_01_l}
    constraintSetup [
     {transformType=Rotation, targetBone {boneName=pelvis}}
     {transformType=Rotation, targetBone {boneName=dyn_skirt_bk_01}}
     {transformType=Rotation, targetBone {boneName=dyn_skirt_bk2_01_l}}
    ]
    constraintWeights [1,1,1]
   }
   n8 = "Constraint - dyn_skirt_bk_02_l (RRR)"!(ComponentPose=@n7, Alpha=1, ConstraintWeights_0=0, ConstraintWeights_1:float=0.5, ConstraintWeights_2:float=0.5) -> (Pose) {
    alphaBoolBlend {blendOption=Linear}
    boneToModify {boneName=dyn_skirt_bk_02_l}
    constraintSetup [
     {transformType=Rotation, targetBone {boneName=pelvis}}
     {transformType=Rotation, targetBone {boneName=dyn_skirt_bk_02}}
     {transformType=Rotation, targetBone {boneName=dyn_skirt_bk2_02_l}}
    ]
    constraintWeights [1,1,1]
   }
   n9 = "Constraint - dyn_skirt_bk_03_l (RRR)"!(ComponentPose=@n8, Alpha=1, ConstraintWeights_0=0, ConstraintWeights_1:float=0.5, ConstraintWeights_2:float=0.5) -> (Pose) {
    alphaBoolBlend {blendOption=Linear}
    boneToModify {boneName=dyn_skirt_bk_03_l}
    constraintSetup [
     {transformType=Rotation, targetBone {boneName=pelvis}}
     {transformType=Rotation, targetBone {boneName=dyn_skirt_bk_03}}
     {transformType=Rotation, targetBone {boneName=dyn_skirt_bk2_03_l}}
    ]
    constraintWeights [1,1,1]
   }
   n10 = "Constraint - dyn_skirt_bk_01_r (RRR)"!(ComponentPose=@n9, Alpha=1, ConstraintWeights_0=0, ConstraintWeights_1:float=0.5, ConstraintWeights_2:float=0.5) -> (Pose) {
    alphaBoolBlend {blendOption=Linear}
    boneToModify {boneName=dyn_skirt_bk_01_r}
    constraintSetup [
     {transformType=Rotation, targetBone {boneName=pelvis}}
     {transformType=Rotation, targetBone {boneName=dyn_skirt_bk_01}}
     {transformType=Rotation, targetBone {boneName=dyn_skirt_bk2_01_r}}
    ]
    constraintWeights [1,1,1]
   }
   n11 = "Constraint - dyn_skirt_bk_02_r (RRR)"!(ComponentPose=@n10, Alpha=1, ConstraintWeights_0=0, ConstraintWeights_1:float=0.5, ConstraintWeights_2:float=0.5) -> (Pose) {
    alphaBoolBlend {blendOption=Linear}
    boneToModify {boneName=dyn_skirt_bk_02_r}
    constraintSetup [
     {transformType=Rotation, targetBone {boneName=pelvis}}
     {transformType=Rotation, targetBone {boneName=dyn_skirt_bk_02}}
     {transformType=Rotation, targetBone {boneName=dyn_skirt_bk2_02_r}}
    ]
    constraintWeights [1,1,1]
   }
   n12 = "Constraint - dyn_skirt_bk_03_r (RRR)"!(ComponentPose=@n11, Alpha=1, ConstraintWeights_0=0, ConstraintWeights_1:float=0.5, ConstraintWeights_2:float=0.5) -> (Pose) {
    alphaBoolBlend {blendOption=Linear}
    boneToModify {boneName=dyn_skirt_bk_03_r}
    constraintSetup [
     {transformType=Rotation, targetBone {boneName=pelvis}}
     {transformType=Rotation, targetBone {boneName=dyn_skirt_bk_03}}
     {transformType=Rotation, targetBone {boneName=dyn_skirt_bk2_03_r}}
    ]
    constraintWeights [1,1,1]
   }
   n13 = "Constraint - dyn_skirt_bk2_01_r (P)"(ComponentPose=@n12, Alpha=1, ConstraintWeights_0=1) {
    alphaBoolBlend {blendOption=Linear}
    boneToModify {boneName=dyn_skirt_bk2_01_r}
    constraintSetup [{targetBone {boneName=pelvis}}]
    constraintWeights [1]
   }
   n14 = "Constraint - dyn_skirt_bk2_02_r (P)"(ComponentPose=@n13, Alpha=1, ConstraintWeights_0=1) {
    alphaBoolBlend {blendOption=Linear}
    boneToModify {boneName=dyn_skirt_bk2_02_r}
    constraintSetup [{targetBone {boneName=pelvis}}]
    constraintWeights [1]
   }
   n15 = "Constraint - dyn_skirt_bk2_03_r (P)"(ComponentPose=@n14, Alpha=1, ConstraintWeights_0=1) {
    alphaBoolBlend {blendOption=Linear}
    boneToModify {boneName=dyn_skirt_bk2_03_r}
    constraintSetup [{targetBone {boneName=pelvis}}]
    constraintWeights [1]
   }
   n16 = Anim Dynamics - Bone: dyn_skirt_bk_01_l(ComponentPose=@n15, Alpha=0.5, ExternalForce="0,0,0") {
    angularDampingOverride=0.9999
    angularSpringConstant=75
    bAngularSpring=true
    bGravityOverrideInSimSpace=true
    bOverrideAngularDamping=true
    simulationSpace=RootRelative
    alphaBoolBlend {blendOption=Linear}
    boundBone {boneName=dyn_skirt_bk_01_l}
    componentLinearAccScale=<1.2 1.2 1.2>
    componentLinearVelScale=<1.2 1.2 1.2>
    physicsBodyDefinitions [{
      sphereCollisionRadius=0
      boundBone {boneName=dyn_skirt_bk_01_l}
      boxExtents=<20 20 10>
      constraintSetup {angularLimitsMax=<25 0 25>, angularLimitsMin=<-25 -25 -25>, angularTarget=<0 0 0>}
      localJointOffset=<2 0 0>
     }]
   }
   n17 = Anim Dynamics - Bone: dyn_skirt_bk_02_l(ComponentPose=@n16, Alpha=0.5, ExternalForce="0,0,0") {
    angularDampingOverride=0.9999
    angularSpringConstant=75
    bAngularSpring=true
    bGravityOverrideInSimSpace=true
    bOverrideAngularDamping=true
    simulationSpace=RootRelative
    alphaBoolBlend {blendOption=Linear}
    boundBone {boneName=dyn_skirt_bk_02_l}
    componentLinearAccScale=<1.2 1.2 1.2>
    componentLinearVelScale=<1.2 1.2 1.2>
    physicsBodyDefinitions [{
      sphereCollisionRadius=0
      boundBone {boneName=dyn_skirt_bk_02_l}
      boxExtents=<20 20 10>
      constraintSetup {angularLimitsMax=<25 0 25>, angularLimitsMin=<-25 -25 -25>, angularTarget=<0 0 0>}
      localJointOffset=<2 0 0>
     }]
   }
   n18 = Anim Dynamics - Bone: dyn_skirt_bk_03_l(ComponentPose=@n17, Alpha=0.5, ExternalForce="0,0,0") {
    angularDampingOverride=0.9999
    angularSpringConstant=75
    bAngularSpring=true
    bGravityOverrideInSimSpace=true
    bOverrideAngularDamping=true
    simulationSpace=RootRelative
    alphaBoolBlend {blendOption=Linear}
    boundBone {boneName=dyn_skirt_bk_03_l}
    componentLinearAccScale=<1.2 1.2 1.2>
    componentLinearVelScale=<1.2 1.2 1.2>
    physicsBodyDefinitions [{
      sphereCollisionRadius=0
      boundBone {boneName=dyn_skirt_bk_03_l}
      boxExtents=<20 20 10>
      constraintSetup {angularLimitsMax=<25 0 25>, angularLimitsMin=<-25 -25 -25>, angularTarget=<0 0 0>}
      localJointOffset=<2 0 0>
     }]
   }
   n19 = Anim Dynamics - Bone: dyn_skirt_bk_01(ComponentPose=@n18, Alpha=0.5, ExternalForce="0,0,0") {
    angularDampingOverride=0.9999
    angularSpringConstant=75
    bAngularSpring=true
    bGravityOverrideInSimSpace=true
    bOverrideAngularDamping=true
    simulationSpace=RootRelative
    alphaBoolBlend {blendOption=Linear}
    boundBone {boneName=dyn_skirt_bk_01}
    componentLinearAccScale=<1.2 1.2 1.2>
    componentLinearVelScale=<1.2 1.2 1.2>
    physicsBodyDefinitions [{
      sphereCollisionRadius=0
      boundBone {boneName=dyn_skirt_bk_01}
      boxExtents=<20 20 10>
      constraintSetup {angularLimitsMax=<25 0 25>, angularLimitsMin=<-25 -25 -25>, angularTarget=<0 0 0>}
      localJointOffset=<2 0 0>
     }]
   }
   n20 = Anim Dynamics - Bone: dyn_skirt_bk_02(ComponentPose=@n19, Alpha=0.5, ExternalForce="0,0,0") {
    angularDampingOverride=0.9999
    angularSpringConstant=75
    bAngularSpring=true
    bGravityOverrideInSimSpace=true
    bOverrideAngularDamping=true
    simulationSpace=RootRelative
    alphaBoolBlend {blendOption=Linear}
    boundBone {boneName=dyn_skirt_bk_02}
    componentLinearAccScale=<1.2 1.2 1.2>
    componentLinearVelScale=<1.2 1.2 1.2>
    physicsBodyDefinitions [{
      sphereCollisionRadius=0
      boundBone {boneName=dyn_skirt_bk_02}
      boxExtents=<20 20 10>
      constraintSetup {angularLimitsMax=<25 0 25>, angularLimitsMin=<-25 -25 -25>, angularTarget=<0 0 0>}
      localJointOffset=<2 0 0>
     }]
   }
   n21 = Anim Dynamics - Bone: dyn_skirt_bk_03(ComponentPose=@n20, Alpha=0.5, ExternalForce="0,0,0") {
    angularDampingOverride=0.9999
    angularSpringConstant=75
    bAngularSpring=true
    bGravityOverrideInSimSpace=true
    bOverrideAngularDamping=true
    simulationSpace=RootRelative
    alphaBoolBlend {blendOption=Linear}
    boundBone {boneName=dyn_skirt_bk_03}
    componentLinearAccScale=<1.2 1.2 1.2>
    componentLinearVelScale=<1.2 1.2 1.2>
    physicsBodyDefinitions [{
      sphereCollisionRadius=0
      boundBone {boneName=dyn_skirt_bk_03}
      boxExtents=<20 20 10>
      constraintSetup {angularLimitsMax=<25 0 25>, angularLimitsMin=<-25 -25 -25>, angularTarget=<0 0 0>}
      localJointOffset=<2 0 0>
     }]
   }
   n22 = Anim Dynamics - Bone: dyn_skirt_bk_01_r(ComponentPose=@n21, Alpha=0.5, ExternalForce="0,0,0") {
    angularDampingOverride=0.9999
    angularSpringConstant=75
    bAngularSpring=true
    bGravityOverrideInSimSpace=true
    bOverrideAngularDamping=true
    simulationSpace=RootRelative
    alphaBoolBlend {blendOption=Linear}
    boundBone {boneName=dyn_skirt_bk_01_r}
    componentLinearAccScale=<1.2 1.2 1.2>
    componentLinearVelScale=<1.2 1.2 1.2>
    physicsBodyDefinitions [{
      sphereCollisionRadius=0
      boundBone {boneName=dyn_skirt_bk_01_r}
      boxExtents=<20 20 10>
      constraintSetup {angularLimitsMax=<25 0 25>, angularLimitsMin=<-25 -25 -25>, angularTarget=<0 0 0>}
      localJointOffset=<2 0 0>
     }]
   }
   n23 = Anim Dynamics - Bone: dyn_skirt_bk_02_r(ComponentPose=@n22, Alpha=0.5, ExternalForce="0,0,0") {
    angularDampingOverride=0.9999
    angularSpringConstant=75
    bAngularSpring=true
    bGravityOverrideInSimSpace=true
    bOverrideAngularDamping=true
    simulationSpace=RootRelative
    alphaBoolBlend {blendOption=Linear}
    boundBone {boneName=dyn_skirt_bk_02_r}
    componentLinearAccScale=<1.2 1.2 1.2>
    componentLinearVelScale=<1.2 1.2 1.2>
    physicsBodyDefinitions [{
      sphereCollisionRadius=0
      boundBone {boneName=dyn_skirt_bk_02_r}
      boxExtents=<20 20 10>
      constraintSetup {angularLimitsMax=<25 0 25>, angularLimitsMin=<-25 -25 -25>, angularTarget=<0 0 0>}
      localJointOffset=<2 0 0>
     }]
   }
   n24 = Anim Dynamics - Bone: dyn_skirt_bk_03_r(ComponentPose=@n23, Alpha=0.5, ExternalForce="0,0,0") {
    angularDampingOverride=0.9999
    angularSpringConstant=75
    bAngularSpring=true
    bGravityOverrideInSimSpace=true
    bOverrideAngularDamping=true
    simulationSpace=RootRelative
    alphaBoolBlend {blendOption=Linear}
    boundBone {boneName=dyn_skirt_bk_03_r}
    componentLinearAccScale=<1.2 1.2 1.2>
    componentLinearVelScale=<1.2 1.2 1.2>
    physicsBodyDefinitions [{
      sphereCollisionRadius=0
      boundBone {boneName=dyn_skirt_bk_03_r}
      boxExtents=<20 20 10>
      constraintSetup {angularLimitsMax=<25 0 25>, angularLimitsMin=<-25 -25 -25>, angularTarget=<0 0 0>}
      localJointOffset=<2 0 0>
     }]
   }
   n25 = Component To Local(ComponentPose=@n24)
   n26 = RigidBody(ComponentPose=Local To Component(LocalPose=Use cached pose 'PreDressRBAN'()), OverrideWorldGravity=Get Gravity_Override_Skirt()) {
    bClampLinearTranslationLimitToRefPose=true
    bOverrideWorldGravity=true
    overridePhysicsAsset=&F_MED_Cyber_Karate_Human_Dress_Physics
    simulationSpace=BaseBoneSpace
    alphaBoolBlend {blendOption=Linear}
    baseBoneRef {boneName=pelvis}
    componentAppliedLinearAccClamp=<1000 1000 1000>
    componentLinearAccScale=<0.25 0.25 0.25>
    overrideWorldGravity {z=-980}
   }
   n27 = Component To Local(ComponentPose=@n26)
   n28 = Get bSkyDivingModeOrSkyDiving() -> (bSkyDivingModeOrSkyDiving:bool)
   n29 = "Constraint - dyn_skirt_fr_01_l_MSTR (PP)"!(
    ComponentPose=Local To Component(LocalPose=Blend Poses by bool(BlendPose_0=@n25, BlendPose_1=@n27, BlendTime_0=0.01, BlendTime_1=0.01, bActiveValue=@n28) {blendType=Linear})
    Alpha=0.9
    ConstraintWeights_0=0
    ConstraintWeights_1:float=1
   ) -> (Pose) {
    alphaBoolBlend {blendOption=Linear}
    boneToModify {boneName=dyn_skirt_fr_01_l_MSTR}
    constraintSetup [{targetBone {boneName=pelvis}}, {targetBone {boneName=dyn_thigh_01_l}}]
    constraintWeights [1,1]
   }
   n30 = "Constraint - dyn_skirt_fr_01_l (PP)"!(ComponentPose=@n29, Alpha=0.9, ConstraintWeights_0=0, ConstraintWeights_1:float=1) -> (Pose) {
    alphaBoolBlend {blendOption=Linear}
    boneToModify {boneName=dyn_skirt_fr_01_l}
    constraintSetup [{targetBone {boneName=pelvis}}, {targetBone {boneName=dyn_thigh_01_l}}]
    constraintWeights [1,1]
   }
   n31 = "Constraint - dyn_skirt_fr_02_l (PP)"!(ComponentPose=@n30, Alpha=0.8, ConstraintWeights_0=0, ConstraintWeights_1:float=1) -> (Pose) {
    alphaBoolBlend {blendOption=Linear}
    boneToModify {boneName=dyn_skirt_fr_02_l}
    constraintSetup [{targetBone {boneName=pelvis}}, {targetBone {boneName=dyn_thigh_01_l}}]
    constraintWeights [1,1]
   }
   n32 = "Constraint - dyn_skirt_fr_03_l (PP)"!(ComponentPose=@n31, Alpha=0.5, ConstraintWeights_0=0, ConstraintWeights_1:float=1) -> (Pose) {
    alphaBoolBlend {blendOption=Linear}
    boneToModify {boneName=dyn_skirt_fr_03_l}
    constraintSetup [{targetBone {boneName=pelvis}}, {targetBone {boneName=dyn_thigh_01_l}}]
    constraintWeights [1,1]
   }
   n33 = "Constraint - dyn_skirt_fr2_01_l_MSTR (PP)"!(ComponentPose=@n32, Alpha=0.9, ConstraintWeights_0=0, ConstraintWeights_1:float=1) -> (Pose) {
    alphaBoolBlend {blendOption=Linear}
    boneToModify {boneName=dyn_skirt_fr2_01_l_MSTR}
    constraintSetup [{targetBone {boneName=pelvis}}, {targetBone {boneName=dyn_thigh_01_l}}]
    constraintWeights [1,1]
   }
   n34 = "Constraint - dyn_skirt_fr2_01_l (PP)"!(ComponentPose=@n33, Alpha=0.9, ConstraintWeights_0=0, ConstraintWeights_1:float=1) -> (Pose) {
    alphaBoolBlend {blendOption=Linear}
    boneToModify {boneName=dyn_skirt_fr2_01_l}
    constraintSetup [{targetBone {boneName=pelvis}}, {targetBone {boneName=dyn_thigh_01_l}}]
    constraintWeights [1,1]
   }
   n35 = "Constraint - dyn_skirt_fr2_02_l (PP)"!(ComponentPose=@n34, Alpha=0.8, ConstraintWeights_0=0, ConstraintWeights_1:float=1) -> (Pose) {
    alphaBoolBlend {blendOption=Linear}
    boneToModify {boneName=dyn_skirt_fr2_02_l}
    constraintSetup [{targetBone {boneName=pelvis}}, {targetBone {boneName=dyn_thigh_01_l}}]
    constraintWeights [1,1]
   }
   n36 = "Constraint - dyn_skirt_fr2_03_l (PP)"!(ComponentPose=@n35, Alpha=0.5, ConstraintWeights_0=0, ConstraintWeights_1:float=1) -> (Pose) {
    alphaBoolBlend {blendOption=Linear}
    boneToModify {boneName=dyn_skirt_fr2_03_l}
    constraintSetup [{targetBone {boneName=pelvis}}, {targetBone {boneName=dyn_thigh_01_l}}]
    constraintWeights [1,1]
   }
   n37 = "Constraint - dyn_skirt_fr_01_r_MSTR (PP)"!(ComponentPose=@n36, Alpha=0.9, ConstraintWeights_0=0, ConstraintWeights_1:float=1) -> (Pose) {
    alphaBoolBlend {blendOption=Linear}
    boneToModify {boneName=dyn_skirt_fr_01_r_MSTR}
    constraintSetup [{targetBone {boneName=pelvis}}, {targetBone {boneName=dyn_thigh_01_r}}]
    constraintWeights [1,1]
   }
   n38 = "Constraint - dyn_skirt_fr_01_r (PP)"!(ComponentPose=@n37, Alpha=0.9, ConstraintWeights_0=0, ConstraintWeights_1:float=1) -> (Pose) {
    alphaBoolBlend {blendOption=Linear}
    boneToModify {boneName=dyn_skirt_fr_01_r}
    constraintSetup [{targetBone {boneName=pelvis}}, {targetBone {boneName=dyn_thigh_01_r}}]
    constraintWeights [1,1]
   }
   n39 = "Constraint - dyn_skirt_fr_02_r (PP)"!(ComponentPose=@n38, Alpha=0.5, ConstraintWeights_0=0, ConstraintWeights_1:float=1) -> (Pose) {
    alphaBoolBlend {blendOption=Linear}
    boneToModify {boneName=dyn_skirt_fr_02_r}
    constraintSetup [{targetBone {boneName=pelvis}}, {targetBone {boneName=dyn_thigh_01_r}}]
    constraintWeights [1,1]
   }
   n40 = "Constraint - dyn_skirt_fr_03_r (PP)"!(ComponentPose=@n39, Alpha=0.5, ConstraintWeights_0=0, ConstraintWeights_1:float=1) -> (Pose) {
    alphaBoolBlend {blendOption=Linear}
    boneToModify {boneName=dyn_skirt_fr_03_r}
    constraintSetup [{targetBone {boneName=pelvis}}, {targetBone {boneName=dyn_thigh_01_r}}]
    constraintWeights [1,1]
   }
   n41 = "Constraint - dyn_skirt_fr2_01_r_MSTR (PP)"!(ComponentPose=@n40, Alpha=0.9, ConstraintWeights_0=0, ConstraintWeights_1:float=1) -> (Pose) {
    alphaBoolBlend {blendOption=Linear}
    boneToModify {boneName=dyn_skirt_fr2_01_r_MSTR}
    constraintSetup [{targetBone {boneName=pelvis}}, {targetBone {boneName=dyn_thigh_01_r}}]
    constraintWeights [1,1]
   }
   n42 = "Constraint - dyn_skirt_fr2_01_r (PP)"!(ComponentPose=@n41, Alpha=0.9, ConstraintWeights_0=0, ConstraintWeights_1:float=1) -> (Pose) {
    alphaBoolBlend {blendOption=Linear}
    boneToModify {boneName=dyn_skirt_fr2_01_r}
    constraintSetup [{targetBone {boneName=pelvis}}, {targetBone {boneName=dyn_thigh_01_r}}]
    constraintWeights [1,1]
   }
   n43 = "Constraint - dyn_skirt_fr2_02_r (PP)"!(ComponentPose=@n42, Alpha=0.5, ConstraintWeights_0=0, ConstraintWeights_1:float=1) -> (Pose) {
    alphaBoolBlend {blendOption=Linear}
    boneToModify {boneName=dyn_skirt_fr2_02_r}
    constraintSetup [{targetBone {boneName=pelvis}}, {targetBone {boneName=dyn_thigh_01_r}}]
    constraintWeights [1,1]
   }
   n44 = "Constraint - dyn_skirt_fr2_03_r (PP)"!(ComponentPose=@n43, Alpha=0.5, ConstraintWeights_0=0, ConstraintWeights_1:float=1) -> (Pose) {
    alphaBoolBlend {blendOption=Linear}
    boneToModify {boneName=dyn_skirt_fr2_03_r}
    constraintSetup [{targetBone {boneName=pelvis}}, {targetBone {boneName=dyn_thigh_01_r}}]
    constraintWeights [1,1]
   }
   n45 = "Constraint - dyn_skirt_fr_01_MSTR (PP)"!(ComponentPose=@n44, Alpha=1, ConstraintWeights_0=0.5, ConstraintWeights_1:float=0.5) -> (Pose) {
    alphaBoolBlend {blendOption=Linear}
    boneToModify {boneName=dyn_skirt_fr_01_MSTR}
    constraintSetup [{targetBone {boneName=dyn_skirt_fr_01_l_MSTR}}, {targetBone {boneName=dyn_skirt_fr_01_r_MSTR}}]
    constraintWeights [1,1]
   }
   n46 = "Constraint - dyn_skirt_fr_01 (PP)"!(ComponentPose=@n45, Alpha=1, ConstraintWeights_0=0.5, ConstraintWeights_1:float=0.5) -> (Pose) {
    alphaBoolBlend {blendOption=Linear}
    boneToModify {boneName=dyn_skirt_fr_01}
    constraintSetup [{targetBone {boneName=dyn_skirt_fr_01_l}}, {targetBone {boneName=dyn_skirt_fr_01_r}}]
    constraintWeights [1,1]
   }
   n47 = "Constraint - dyn_skirt_fr_02 (PP)"!(ComponentPose=@n46, Alpha=1, ConstraintWeights_0=0.5, ConstraintWeights_1:float=0.5) -> (Pose) {
    alphaBoolBlend {blendOption=Linear}
    boneToModify {boneName=dyn_skirt_fr_02}
    constraintSetup [{targetBone {boneName=dyn_skirt_fr_02_l}}, {targetBone {boneName=dyn_skirt_fr_02_r}}]
    constraintWeights [1,1]
   }
   n48 = "Constraint - dyn_skirt_fr_03 (PP)"!(ComponentPose=@n47, Alpha=1, ConstraintWeights_0=0.5, ConstraintWeights_1:float=0.5) -> (Pose) {
    alphaBoolBlend {blendOption=Linear}
    boneToModify {boneName=dyn_skirt_fr_03}
    constraintSetup [{targetBone {boneName=dyn_skirt_fr_03_l}}, {targetBone {boneName=dyn_skirt_fr_03_r}}]
    constraintWeights [1,1]
   }
   n49 = "Constraint - dyn_skirt_01_l_MSTR (P)"(ComponentPose=@n48, Alpha=0.5, ConstraintWeights_0=1) {
    alphaBoolBlend {blendOption=Linear}
    boneToModify {boneName=dyn_skirt_01_l_MSTR}
    constraintSetup [{targetBone {boneName=dyn_skirt_fr2_01_l_MSTR}}]
    constraintWeights [1]
   }
   n50 = "Constraint - dyn_skirt_01_l (P)"(ComponentPose=@n49, Alpha=0.5, ConstraintWeights_0=1) {
    alphaBoolBlend {blendOption=Linear}
    boneToModify {boneName=dyn_skirt_01_l}
    constraintSetup [{targetBone {boneName=dyn_skirt_fr2_01_l}}]
    constraintWeights [1]
   }
   n51 = "Constraint - dyn_skirt_02_l (P)"(ComponentPose=@n50, Alpha=0.5, ConstraintWeights_0=1) {
    alphaBoolBlend {blendOption=Linear}
    boneToModify {boneName=dyn_skirt_02_l}
    constraintSetup [{targetBone {boneName=dyn_skirt_fr2_02_l}}]
    constraintWeights [1]
   }
   n52 = "Constraint - dyn_skirt_03_l (P)"(ComponentPose=@n51, Alpha=0.5, ConstraintWeights_0=1) {
    alphaBoolBlend {blendOption=Linear}
    boneToModify {boneName=dyn_skirt_03_l}
    constraintSetup [{targetBone {boneName=dyn_skirt_fr2_03_l}}]
    constraintWeights [1]
   }
   n53 = "Constraint - dyn_skirt_bk2_01_l_MSTR (P)"(ComponentPose=@n52, Alpha=0.5, ConstraintWeights_0=1) {
    alphaBoolBlend {blendOption=Linear}
    boneToModify {boneName=dyn_skirt_bk2_01_l_MSTR}
    constraintSetup [{targetBone {boneName=dyn_skirt_01_l_MSTR}}]
    constraintWeights [1]
   }
   n54 = "Constraint - dyn_skirt_bk2_01_l (P)"(ComponentPose=@n53, Alpha=0.5, ConstraintWeights_0=1) {
    alphaBoolBlend {blendOption=Linear}
    boneToModify {boneName=dyn_skirt_bk2_01_l}
    constraintSetup [{targetBone {boneName=dyn_skirt_01_l}}]
    constraintWeights [1]
   }
   n55 = "Constraint - dyn_skirt_bk2_02_l (P)"(ComponentPose=@n54, Alpha=0.5, ConstraintWeights_0=1) {
    alphaBoolBlend {blendOption=Linear}
    boneToModify {boneName=dyn_skirt_bk2_02_l}
    constraintSetup [{targetBone {boneName=dyn_skirt_02_l}}]
    constraintWeights [1]
   }
   n56 = "Constraint - dyn_skirt_bk2_03_l (P)"(ComponentPose=@n55, Alpha=0.5, ConstraintWeights_0=1) {
    alphaBoolBlend {blendOption=Linear}
    boneToModify {boneName=dyn_skirt_bk2_03_l}
    constraintSetup [{targetBone {boneName=dyn_skirt_03_l}}]
    constraintWeights [1]
   }
   n57 = "Constraint - dyn_skirt_01_r_MSTR (P)"(ComponentPose=@n56, Alpha=0.5, ConstraintWeights_0=1) {
    alphaBoolBlend {blendOption=Linear}
    boneToModify {boneName=dyn_skirt_01_r_MSTR}
    constraintSetup [{targetBone {boneName=dyn_skirt_fr2_01_r_MSTR}}]
    constraintWeights [1]
   }
   n58 = "Constraint - dyn_skirt_01_r (P)"(ComponentPose=@n57, Alpha=0.5, ConstraintWeights_0=1) {
    alphaBoolBlend {blendOption=Linear}
    boneToModify {boneName=dyn_skirt_01_r}
    constraintSetup [{targetBone {boneName=dyn_skirt_fr2_01_r}}]
    constraintWeights [1]
   }
   n59 = "Constraint - dyn_skirt_02_r (P)"(ComponentPose=@n58, Alpha=0.5, ConstraintWeights_0=1) {
    alphaBoolBlend {blendOption=Linear}
    boneToModify {boneName=dyn_skirt_02_r}
    constraintSetup [{targetBone {boneName=dyn_skirt_fr2_02_r}}]
    constraintWeights [1]
   }
   n60 = "Constraint - dyn_skirt_03_r (P)"(ComponentPose=@n59, Alpha=0.5, ConstraintWeights_0=1) {
    alphaBoolBlend {blendOption=Linear}
    boneToModify {boneName=dyn_skirt_03_r}
    constraintSetup [{targetBone {boneName=dyn_skirt_fr2_03_r}}]
    constraintWeights [1]
   }
   n61 = "Constraint - dyn_skirt_bk2_01_r_MSTR (P)"(ComponentPose=@n60, Alpha=0.5, ConstraintWeights_0=1) {
    alphaBoolBlend {blendOption=Linear}
    boneToModify {boneName=dyn_skirt_bk2_01_r_MSTR}
    constraintSetup [{targetBone {boneName=dyn_skirt_01_r_MSTR}}]
    constraintWeights [1]
   }
   n62 = "Constraint - dyn_skirt_bk2_01_r (P)"(ComponentPose=@n61, Alpha=0.5, ConstraintWeights_0=1) {
    alphaBoolBlend {blendOption=Linear}
    boneToModify {boneName=dyn_skirt_bk2_01_r}
    constraintSetup [{targetBone {boneName=dyn_skirt_01_r}}]
    constraintWeights [1]
   }
   n63 = "Constraint - dyn_skirt_bk2_02_r (P)"(ComponentPose=@n62, Alpha=0.5, ConstraintWeights_0=1) {
    alphaBoolBlend {blendOption=Linear}
    boneToModify {boneName=dyn_skirt_bk2_02_r}
    constraintSetup [{targetBone {boneName=dyn_skirt_02_r}}]
    constraintWeights [1]
   }
   n64 = "Constraint - dyn_skirt_bk2_03_r (P)"(ComponentPose=@n63, Alpha=0.5, ConstraintWeights_0=1) {
    alphaBoolBlend {blendOption=Linear}
    boneToModify {boneName=dyn_skirt_bk2_03_r}
    constraintSetup [{targetBone {boneName=dyn_skirt_03_r}}]
    constraintWeights [1]
   }
   n65 = "Constraint - dyn_skirt_bk2_01_l_MSTR (PP)"!(ComponentPose=@n64, Alpha=1, ConstraintWeights_0=0.5, ConstraintWeights_1:float=0.5) -> (Pose) {
    alphaBoolBlend {blendOption=Linear}
    boneToModify {boneName=dyn_skirt_bk2_01_l_MSTR}
    constraintSetup [{targetBone {boneName=dyn_skirt_bk_01_l_MSTR}}, {targetBone {boneName=dyn_skirt_01_l_MSTR}}]
    constraintWeights [1,1]
   }
   n66 = "Constraint - dyn_skirt_bk2_01_l (PP)"!(ComponentPose=@n65, Alpha=1, ConstraintWeights_0=0.5, ConstraintWeights_1:float=0.5) -> (Pose) {
    alphaBoolBlend {blendOption=Linear}
    boneToModify {boneName=dyn_skirt_bk2_01_l}
    constraintSetup [{targetBone {boneName=dyn_skirt_bk_01_l}}, {targetBone {boneName=dyn_skirt_01_l}}]
    constraintWeights [1,1]
   }
   n67 = "Constraint - dyn_skirt_bk2_02_l (PP)"!(ComponentPose=@n66, Alpha=1, ConstraintWeights_0=0.5, ConstraintWeights_1:float=0.5) -> (Pose) {
    alphaBoolBlend {blendOption=Linear}
    boneToModify {boneName=dyn_skirt_bk2_02_l}
    constraintSetup [{targetBone {boneName=dyn_skirt_bk_02_l}}, {targetBone {boneName=dyn_skirt_02_l}}]
    constraintWeights [1,1]
   }
   n68 = "Constraint - dyn_skirt_bk2_03_l (PP)"!(ComponentPose=@n67, Alpha=1, ConstraintWeights_0=0.5, ConstraintWeights_1:float=0.5) -> (Pose) {
    alphaBoolBlend {blendOption=Linear}
    boneToModify {boneName=dyn_skirt_bk2_03_l}
    constraintSetup [{targetBone {boneName=dyn_skirt_bk_03_l}}, {targetBone {boneName=dyn_skirt_03_l}}]
    constraintWeights [1,1]
   }
   n69 = "Transform (Modify) Bone - Bone: dyn_skirt_bk2_01_l_MSTR"(ComponentPose=@n68, Alpha=1, Translation="2,0,0", Rotation="0,0,10", Scale="1,1,1") {
    rotationMode=BMM_Additive
    scaleMode=BMM_Additive
    translationMode=BMM_Additive
    alphaBoolBlend {blendOption=Linear}
    boneToModify {boneName=dyn_skirt_bk2_01_l_MSTR}
    rotation {roll=10}
    translation {x=2}
   }
   n70 = "Constraint - dyn_skirt_01_l_MSTR (PP)"!(ComponentPose=@n69, Alpha=1, ConstraintWeights_0=0.5, ConstraintWeights_1:float=0.5) -> (Pose) {
    alphaBoolBlend {blendOption=Linear}
    boneToModify {boneName=dyn_skirt_01_l_MSTR}
    constraintSetup [{targetBone {boneName=dyn_skirt_bk2_01_l_MSTR}}, {targetBone {boneName=dyn_skirt_fr2_01_l_MSTR}}]
    constraintWeights [1,1]
   }
   n71 = "Constraint - dyn_skirt_01_l (PP)"!(ComponentPose=@n70, Alpha=1, ConstraintWeights_0=0.5, ConstraintWeights_1:float=0.5) -> (Pose) {
    alphaBoolBlend {blendOption=Linear}
    boneToModify {boneName=dyn_skirt_01_l}
    constraintSetup [{targetBone {boneName=dyn_skirt_bk2_01_l}}, {targetBone {boneName=dyn_skirt_fr2_01_l_MSTR}}]
    constraintWeights [1,1]
   }
   n72 = "Constraint - dyn_skirt_02_l (PP)"!(ComponentPose=@n71, Alpha=1, ConstraintWeights_0=0.5, ConstraintWeights_1:float=0.5) -> (Pose) {
    alphaBoolBlend {blendOption=Linear}
    boneToModify {boneName=dyn_skirt_02_l}
    constraintSetup [{targetBone {boneName=dyn_skirt_bk2_02_l}}, {targetBone {boneName=dyn_skirt_fr2_02_l}}]
    constraintWeights [1,1]
   }
   n73 = "Constraint - dyn_skirt_03_l (PP)"!(ComponentPose=@n72, Alpha=1, ConstraintWeights_0=0.5, ConstraintWeights_1:float=0.5) -> (Pose) {
    alphaBoolBlend {blendOption=Linear}
    boneToModify {boneName=dyn_skirt_03_l}
    constraintSetup [{targetBone {boneName=dyn_skirt_bk2_03_l}}, {targetBone {boneName=dyn_skirt_fr2_03_l}}]
    constraintWeights [1,1]
   }
   n74 = "Transform (Modify) Bone - Bone: dyn_skirt_01_l_MSTR"(ComponentPose=@n73, Alpha=1, Translation="0,0,3", Rotation="0,0,0", Scale="1,1,1") {
    rotationMode=BMM_Additive
    rotationSpace=BCS_BoneSpace
    translationMode=BMM_Additive
    alphaBoolBlend {blendOption=Linear}
    boneToModify {boneName=dyn_skirt_01_l_MSTR}
    translation {z=3}
   }
   n75 = "Constraint - dyn_skirt_bk2_01_r_MSTR (PP)"!(ComponentPose=@n74, Alpha=1, ConstraintWeights_0=0.5, ConstraintWeights_1:float=0.5) -> (Pose) {
    alphaBoolBlend {blendOption=Linear}
    boneToModify {boneName=dyn_skirt_bk2_01_r_MSTR}
    constraintSetup [{targetBone {boneName=dyn_skirt_bk_01_r_MSTR}}, {targetBone {boneName=dyn_skirt_01_r_MSTR}}]
    constraintWeights [1,1]
   }
   n76 = "Constraint - dyn_skirt_bk2_01_r (PP)"!(ComponentPose=@n75, Alpha=1, ConstraintWeights_0=0.5, ConstraintWeights_1:float=0.5) -> (Pose) {
    alphaBoolBlend {blendOption=Linear}
    boneToModify {boneName=dyn_skirt_bk2_01_r}
    constraintSetup [{targetBone {boneName=dyn_skirt_bk_01_r}}, {targetBone {boneName=dyn_skirt_01_r}}]
    constraintWeights [1,1]
   }
   n77 = "Constraint - dyn_skirt_bk2_02_r (PP)"!(ComponentPose=@n76, Alpha=1, ConstraintWeights_0=0.5, ConstraintWeights_1:float=0.5) -> (Pose) {
    alphaBoolBlend {blendOption=Linear}
    boneToModify {boneName=dyn_skirt_bk2_02_r}
    constraintSetup [{targetBone {boneName=dyn_skirt_bk_02_r}}, {targetBone {boneName=dyn_skirt_02_r}}]
    constraintWeights [1,1]
   }
   n78 = "Constraint - dyn_skirt_bk2_03_r (PP)"!(ComponentPose=@n77, Alpha=1, ConstraintWeights_0=0.5, ConstraintWeights_1:float=0.5) -> (Pose) {
    alphaBoolBlend {blendOption=Linear}
    boneToModify {boneName=dyn_skirt_bk2_03_r}
    constraintSetup [{targetBone {boneName=dyn_skirt_bk_03_r}}, {targetBone {boneName=dyn_skirt_03_r}}]
    constraintWeights [1,1]
   }
   n79 = "Constraint - dyn_skirt_01_r_MSTR (PP)"!(ComponentPose=@n78, Alpha=1, ConstraintWeights_0=0.5, ConstraintWeights_1:float=0.5) -> (Pose) {
    alphaBoolBlend {blendOption=Linear}
    boneToModify {boneName=dyn_skirt_01_r_MSTR}
    constraintSetup [{targetBone {boneName=dyn_skirt_bk2_01_r_MSTR}}, {targetBone {boneName=dyn_skirt_fr2_01_r_MSTR}}]
    constraintWeights [1,1]
   }
   n80 = "Constraint - dyn_skirt_01_r (PP)"!(ComponentPose=@n79, Alpha=1, ConstraintWeights_0=0.5, ConstraintWeights_1:float=0.5) -> (Pose) {
    alphaBoolBlend {blendOption=Linear}
    boneToModify {boneName=dyn_skirt_01_r}
    constraintSetup [{targetBone {boneName=dyn_skirt_bk2_01_r}}, {targetBone {boneName=dyn_skirt_fr2_01_r}}]
    constraintWeights [1,1]
   }
   n81 = "Constraint - dyn_skirt_02_r (PP)"!(ComponentPose=@n80, Alpha=1, ConstraintWeights_0=0.5, ConstraintWeights_1:float=0.5) -> (Pose) {
    alphaBoolBlend {blendOption=Linear}
    boneToModify {boneName=dyn_skirt_02_r}
    constraintSetup [{targetBone {boneName=dyn_skirt_bk2_02_r}}, {targetBone {boneName=dyn_skirt_fr2_02_r}}]
    constraintWeights [1,1]
   }
   n82 = "Constraint - dyn_skirt_03_r (PP)"!(ComponentPose=@n81, Alpha=1, ConstraintWeights_0=0.5, ConstraintWeights_1:float=0.5) -> (Pose) {
    alphaBoolBlend {blendOption=Linear}
    boneToModify {boneName=dyn_skirt_03_r}
    constraintSetup [{targetBone {boneName=dyn_skirt_bk2_03_r}}, {targetBone {boneName=dyn_skirt_fr2_03_r}}]
    constraintWeights [1,1]
   }
   n83 = Get Gravity_Override_Skirt()
   n84 = Anim Dynamics - Bone: dyn_skirt_fr_02_l!(ComponentPose=@n82, Alpha=1, GravityOverride:Vector=@n83, ExternalForce="0,0,0") -> (Pose) {
    angularDampingOverride=0.9999
    angularSpringConstant=75
    bAngularSpring=true
    bGravityOverrideInSimSpace=true
    bOverrideAngularDamping=true
    bUseGravityOverride=true
    simulationSpace=RootRelative
    alphaBoolBlend {blendOption=Linear}
    boundBone {boneName=dyn_skirt_fr_02_l}
    componentAppliedLinearAccClamp=<2500 2500 2500>
    componentLinearAccScale=<0.8 0.8 0.8>
    componentLinearVelScale=<0.8 0.8 0.8>
    physicsBodyDefinitions [{
      sphereCollisionRadius=0
      boundBone {boneName=dyn_skirt_fr_02_l}
      boxExtents=<20 10 10>
      constraintSetup {angularLimitsMax=<25 0 25>, angularLimitsMin=<-25 -25 0>, angularTarget=<0 0 0>}
      localJointOffset=<2 0 0>
     }]
   }
   n85 = Anim Dynamics - Bone: dyn_skirt_fr_03_l!(ComponentPose=@n84, Alpha=1, GravityOverride:Vector=@n83, ExternalForce="0,0,0") -> (Pose) {
    angularDampingOverride=0.9999
    angularSpringConstant=75
    bAngularSpring=true
    bGravityOverrideInSimSpace=true
    bOverrideAngularDamping=true
    bUseGravityOverride=true
    simulationSpace=RootRelative
    alphaBoolBlend {blendOption=Linear}
    boundBone {boneName=dyn_skirt_fr_03_l}
    componentAppliedLinearAccClamp=<2500 2500 2500>
    componentLinearAccScale=<0.8 0.8 0.8>
    componentLinearVelScale=<0.8 0.8 0.8>
    physicsBodyDefinitions [{
      sphereCollisionRadius=0
      boundBone {boneName=dyn_skirt_fr_03_l}
      boxExtents=<20 10 10>
      constraintSetup {angularLimitsMax=<25 0 25>, angularLimitsMin=<-25 -25 -25>, angularTarget=<0 0 0>}
      localJointOffset=<2 0 0>
     }]
   }
   n86 = Anim Dynamics - Bone: dyn_skirt_fr_02_r!(ComponentPose=@n85, Alpha=0.7, GravityOverride:Vector=@n83, ExternalForce="0,0,0") -> (Pose) {
    angularDampingOverride=0.9999
    angularSpringConstant=75
    bAngularSpring=true
    bGravityOverrideInSimSpace=true
    bOverrideAngularDamping=true
    bUseGravityOverride=true
    simulationSpace=RootRelative
    alphaBoolBlend {blendOption=Linear}
    boundBone {boneName=dyn_skirt_fr_02_r}
    componentAppliedLinearAccClamp=<2500 2500 2500>
    componentLinearAccScale=<0.8 0.8 0.8>
    componentLinearVelScale=<0.8 0.8 0.8>
    physicsBodyDefinitions [{
      sphereCollisionRadius=0
      boundBone {boneName=dyn_skirt_fr_02_r}
      boxExtents=<20 20 10>
      constraintSetup {angularLimitsMax=<25 0 25>, angularLimitsMin=<-25 -25 -25>, angularTarget=<0 0 0>}
      localJointOffset=<2 0 0>
     }]
   }
   n87 = Anim Dynamics - Bone: dyn_skirt_fr_03_r!(ComponentPose=@n86, Alpha=0.7, GravityOverride:Vector=@n83, ExternalForce="0,0,0") -> (Pose) {
    angularDampingOverride=0.9999
    angularSpringConstant=75
    bAngularSpring=true
    bGravityOverrideInSimSpace=true
    bOverrideAngularDamping=true
    bUseGravityOverride=true
    simulationSpace=RootRelative
    alphaBoolBlend {blendOption=Linear}
    boundBone {boneName=dyn_skirt_fr_03_r}
    componentAppliedLinearAccClamp=<2500 2500 2500>
    componentLinearAccScale=<0.8 0.8 0.8>
    componentLinearVelScale=<0.8 0.8 0.8>
    physicsBodyDefinitions [{
      sphereCollisionRadius=0
      boundBone {boneName=dyn_skirt_fr_03_r}
      boxExtents=<20 20 10>
      constraintSetup {angularLimitsMax=<25 0 25>, angularLimitsMin=<-25 -25 -25>, angularTarget=<0 0 0>}
      localJointOffset=<2 0 0>
     }]
   }
   n88 = Anim Dynamics - Bone: dyn_skirt_fr2_02_l!(ComponentPose=@n87, Alpha=1, GravityOverride:Vector=@n83, ExternalForce="0,0,0") -> (Pose) {
    angularDampingOverride=0.9999
    angularSpringConstant=75
    bAngularSpring=true
    bGravityOverrideInSimSpace=true
    bOverrideAngularDamping=true
    bUseGravityOverride=true
    simulationSpace=RootRelative
    alphaBoolBlend {blendOption=Linear}
    boundBone {boneName=dyn_skirt_fr2_02_l}
    componentAppliedLinearAccClamp=<2500 2500 2500>
    componentLinearAccScale=<0.8 0.8 0.8>
    componentLinearVelScale=<0.8 0.8 0.8>
    physicsBodyDefinitions [{
      sphereCollisionRadius=0
      boundBone {boneName=dyn_skirt_fr2_02_l}
      boxExtents=<20 10 10>
      constraintSetup {angularLimitsMax=<25 0 25>, angularLimitsMin=<-25 -25 -25>, angularTarget=<0 0 0>}
      localJointOffset=<2 0 0>
     }]
   }
   n89 = Anim Dynamics - Bone: dyn_skirt_fr2_03_l!(ComponentPose=@n88, Alpha=1, GravityOverride:Vector=@n83, ExternalForce="0,0,0") -> (Pose) {
    angularDampingOverride=0.9999
    angularSpringConstant=75
    bAngularSpring=true
    bGravityOverrideInSimSpace=true
    bOverrideAngularDamping=true
    bUseGravityOverride=true
    simulationSpace=RootRelative
    alphaBoolBlend {blendOption=Linear}
    boundBone {boneName=dyn_skirt_fr2_03_l}
    componentAppliedLinearAccClamp=<2500 2500 2500>
    componentLinearAccScale=<0.8 0.8 0.8>
    componentLinearVelScale=<0.8 0.8 0.8>
    physicsBodyDefinitions [{
      sphereCollisionRadius=0
      boundBone {boneName=dyn_skirt_fr2_03_l}
      boxExtents=<20 10 10>
      constraintSetup {angularLimitsMax=<25 0 25>, angularLimitsMin=<-25 -25 -25>, angularTarget=<0 0 0>}
      localJointOffset=<2 0 0>
     }]
   }
   n90 = Anim Dynamics - Bone: dyn_skirt_fr2_02_r!(ComponentPose=@n89, Alpha=0.7, GravityOverride:Vector=@n83, ExternalForce="0,0,0") -> (Pose) {
    angularDampingOverride=0.9999
    angularSpringConstant=75
    bAngularSpring=true
    bGravityOverrideInSimSpace=true
    bOverrideAngularDamping=true
    bUseGravityOverride=true
    simulationSpace=RootRelative
    alphaBoolBlend {blendOption=Linear}
    boundBone {boneName=dyn_skirt_fr2_02_r}
    componentAppliedLinearAccClamp=<2500 2500 2500>
    componentLinearAccScale=<0.8 0.8 0.8>
    componentLinearVelScale=<0.8 0.8 0.8>
    physicsBodyDefinitions [{
      sphereCollisionRadius=0
      boundBone {boneName=dyn_skirt_fr2_02_r}
      boxExtents=<20 20 10>
      constraintSetup {angularLimitsMax=<25 0 25>, angularLimitsMin=<-25 -25 -25>, angularTarget=<0 0 0>}
      localJointOffset=<2 0 0>
     }]
   }
   n91 = Anim Dynamics - Bone: dyn_skirt_fr2_03_r!(ComponentPose=@n90, Alpha=0.7, GravityOverride:Vector=@n83, ExternalForce="0,0,0") -> (Pose) {
    angularDampingOverride=0.9999
    angularSpringConstant=75
    bAngularSpring=true
    bGravityOverrideInSimSpace=true
    bOverrideAngularDamping=true
    bUseGravityOverride=true
    simulationSpace=RootRelative
    alphaBoolBlend {blendOption=Linear}
    boundBone {boneName=dyn_skirt_fr2_03_r}
    componentAppliedLinearAccClamp=<2500 2500 2500>
    componentLinearAccScale=<0.8 0.8 0.8>
    componentLinearVelScale=<0.8 0.8 0.8>
    physicsBodyDefinitions [{
      sphereCollisionRadius=0
      boundBone {boneName=dyn_skirt_fr2_03_r}
      boxExtents=<20 20 10>
      constraintSetup {angularLimitsMax=<25 0 25>, angularLimitsMin=<-25 -25 -25>, angularTarget=<0 0 0>}
      localJointOffset=<2 0 0>
     }]
   }
   n92 = Anim Dynamics - Bone: dyn_skirt_02_l!(ComponentPose=@n91, Alpha=1, GravityOverride:Vector=@n83, ExternalForce="0,0,0") -> (Pose) {
    angularDampingOverride=0.9999
    angularSpringConstant=75
    bAngularSpring=true
    bGravityOverrideInSimSpace=true
    bOverrideAngularDamping=true
    bUseGravityOverride=true
    simulationSpace=RootRelative
    alphaBoolBlend {blendOption=Linear}
    boundBone {boneName=dyn_skirt_02_l}
    componentAppliedLinearAccClamp=<2500 2500 2500>
    componentLinearAccScale=<0.8 0.8 0.8>
    componentLinearVelScale=<0.8 0.8 0.8>
    physicsBodyDefinitions [{
      sphereCollisionRadius=0
      boundBone {boneName=dyn_skirt_02_l}
      boxExtents=<20 10 10>
      constraintSetup {angularLimitsMax=<25 0 25>, angularLimitsMin=<-25 -25 -25>, angularTarget=<0 0 0>}
      localJointOffset=<2 0 0>
     }]
   }
   n93 = Anim Dynamics - Bone: dyn_skirt_03_l!(ComponentPose=@n92, Alpha=1, GravityOverride:Vector=@n83, ExternalForce="0,0,0") -> (Pose) {
    angularDampingOverride=0.9999
    angularSpringConstant=75
    bAngularSpring=true
    bGravityOverrideInSimSpace=true
    bOverrideAngularDamping=true
    bUseGravityOverride=true
    simulationSpace=RootRelative
    alphaBoolBlend {blendOption=Linear}
    boundBone {boneName=dyn_skirt_03_l}
    componentAppliedLinearAccClamp=<2500 2500 2500>
    componentLinearAccScale=<0.8 0.8 0.8>
    componentLinearVelScale=<0.8 0.8 0.8>
    physicsBodyDefinitions [{
      sphereCollisionRadius=0
      boundBone {boneName=dyn_skirt_03_l}
      boxExtents=<20 10 10>
      constraintSetup {angularLimitsMax=<25 0 25>, angularLimitsMin=<-25 -25 -25>, angularTarget=<0 0 0>}
      localJointOffset=<2 0 0>
     }]
   }
   n94 = Anim Dynamics - Bone: dyn_skirt_01_r!(ComponentPose=@n93, Alpha=1, GravityOverride:Vector=@n83, ExternalForce="0,0,0") -> (Pose) {
    angularDampingOverride=0.9999
    angularSpringConstant=75
    bAngularSpring=true
    bGravityOverrideInSimSpace=true
    bOverrideAngularDamping=true
    bUseGravityOverride=true
    simulationSpace=RootRelative
    alphaBoolBlend {blendOption=Linear}
    boundBone {boneName=dyn_skirt_01_r}
    componentAppliedLinearAccClamp=<2500 2500 2500>
    componentLinearAccScale=<0.8 0.8 0.8>
    componentLinearVelScale=<0.8 0.8 0.8>
    physicsBodyDefinitions [{
      sphereCollisionRadius=0
      boundBone {boneName=dyn_skirt_01_r}
      boxExtents=<20 20 10>
      constraintSetup {angularLimitsMax=<25 0 25>, angularLimitsMin=<-25 -25 -25>, angularTarget=<0 0 0>}
      localJointOffset=<2 0 0>
     }]
   }
   n95 = Anim Dynamics - Bone: dyn_skirt_03_r!(ComponentPose=@n94, Alpha=1, GravityOverride:Vector=@n83, ExternalForce="0,0,0") -> (Pose) {
    angularDampingOverride=0.9999
    angularSpringConstant=75
    bAngularSpring=true
    bGravityOverrideInSimSpace=true
    bOverrideAngularDamping=true
    bUseGravityOverride=true
    simulationSpace=RootRelative
    alphaBoolBlend {blendOption=Linear}
    boundBone {boneName=dyn_skirt_03_r}
    componentAppliedLinearAccClamp=<2500 2500 2500>
    componentLinearAccScale=<0.8 0.8 0.8>
    componentLinearVelScale=<0.8 0.8 0.8>
    physicsBodyDefinitions [{
      sphereCollisionRadius=0
      boundBone {boneName=dyn_skirt_03_r}
      boxExtents=<20 20 10>
      constraintSetup {angularLimitsMax=<25 0 25>, angularLimitsMin=<-25 -25 -25>, angularTarget=<0 0 0>}
      localJointOffset=<2 0 0>
     }]
   }
   n96 = Anim Dynamics - Bone: dyn_skirt_bk2_02_r!(ComponentPose=@n95, Alpha=1, GravityOverride:Vector=@n83, ExternalForce="0,0,0") -> (Pose) {
    angularDampingOverride=0.9999
    angularSpringConstant=75
    bAngularSpring=true
    bGravityOverrideInSimSpace=true
    bOverrideAngularDamping=true
    bUseGravityOverride=true
    simulationSpace=RootRelative
    alphaBoolBlend {blendOption=Linear}
    boundBone {boneName=dyn_skirt_bk2_02_r}
    componentAppliedLinearAccClamp=<2500 2500 2500>
    componentLinearAccScale=<0.8 0.8 0.8>
    componentLinearVelScale=<0.8 0.8 0.8>
    physicsBodyDefinitions [{
      sphereCollisionRadius=0
      boundBone {boneName=dyn_skirt_bk2_02_r}
      boxExtents=<20 20 10>
      constraintSetup {angularLimitsMax=<25 0 25>, angularLimitsMin=<-25 -25 -25>, angularTarget=<0 0 0>}
      localJointOffset=<2 0 0>
     }]
   }
   n97 = Anim Dynamics - Bone: dyn_skirt_bk2_02_l!(ComponentPose=@n96, Alpha=1, GravityOverride:Vector=@n83, ExternalForce="0,0,0") -> (Pose) {
    angularDampingOverride=0.9999
    angularSpringConstant=75
    bAngularSpring=true
    bGravityOverrideInSimSpace=true
    bOverrideAngularDamping=true
    bUseGravityOverride=true
    simulationSpace=RootRelative
    alphaBoolBlend {blendOption=Linear}
    boundBone {boneName=dyn_skirt_bk2_02_l}
    componentAppliedLinearAccClamp=<2500 2500 2500>
    componentLinearAccScale=<0.8 0.8 0.8>
    componentLinearVelScale=<0.8 0.8 0.8>
    physicsBodyDefinitions [{
      sphereCollisionRadius=0
      boundBone {boneName=dyn_skirt_bk2_02_l}
      boxExtents=<20 20 10>
      constraintSetup {angularLimitsMax=<25 0 25>, angularLimitsMin=<-25 -25 -25>, angularTarget=<0 0 0>}
      localJointOffset=<2 0 0>
     }]
   }
   n98 = Anim Dynamics - Bone: dyn_skirt_bk2_03_r!(ComponentPose=@n97, Alpha=1, GravityOverride:Vector=@n83, ExternalForce="0,0,0") -> (Pose) {
    angularDampingOverride=0.9999
    angularSpringConstant=75
    bAngularSpring=true
    bGravityOverrideInSimSpace=true
    bOverrideAngularDamping=true
    bUseGravityOverride=true
    simulationSpace=RootRelative
    alphaBoolBlend {blendOption=Linear}
    boundBone {boneName=dyn_skirt_bk2_03_r}
    componentAppliedLinearAccClamp=<2500 2500 2500>
    componentLinearAccScale=<0.8 0.8 0.8>
    componentLinearVelScale=<0.8 0.8 0.8>
    physicsBodyDefinitions [{
      sphereCollisionRadius=0
      boundBone {boneName=dyn_skirt_bk2_03_r}
      boxExtents=<20 20 10>
      constraintSetup {angularLimitsMax=<25 0 25>, angularLimitsMin=<-25 -25 -25>, angularTarget=<0 0 0>}
      localJointOffset=<2 0 0>
     }]
   }
   n99 = Anim Dynamics - Bone: dyn_skirt_bk2_03_l!(ComponentPose=@n98, Alpha=1, GravityOverride:Vector=@n83, ExternalForce="0,0,0") -> (Pose) {
    angularDampingOverride=0.9999
    angularSpringConstant=75
    bAngularSpring=true
    bGravityOverrideInSimSpace=true
    bOverrideAngularDamping=true
    bUseGravityOverride=true
    simulationSpace=RootRelative
    alphaBoolBlend {blendOption=Linear}
    boundBone {boneName=dyn_skirt_bk2_03_l}
    componentAppliedLinearAccClamp=<2500 2500 2500>
    componentLinearAccScale=<0.8 0.8 0.8>
    componentLinearVelScale=<0.8 0.8 0.8>
    physicsBodyDefinitions [{
      sphereCollisionRadius=0
      boundBone {boneName=dyn_skirt_bk2_03_l}
      boxExtents=<20 20 10>
      constraintSetup {angularLimitsMax=<25 0 25>, angularLimitsMin=<-25 -25 -25>, angularTarget=<0 0 0>}
      localJointOffset=<2 0 0>
     }]
   }
   n100 = "Constraint - const_pouch_r (PP)"!(ComponentPose=@n99, Alpha=1, ConstraintWeights_0=0.885, ConstraintWeights_1:float=0.115) -> (Pose) {
    alphaBoolBlend {blendOption=Linear}
    boneToModify {boneName=const_pouch_r}
    constraintSetup [{targetBone {boneName=thigh_r}}, {targetBone {boneName=thigh_twist_01_r}}]
    constraintWeights [1,1]
   }
   n101 = RigidBody!(ComponentPose=@n100, Alpha, ExternalForce) -> (Pose) {
    bClampLinearTranslationLimitToRefPose=true
    bOverrideWorldGravity=true
    overridePhysicsAsset=&/F_MED_Cyber_Karate_Human_Scarf_Physics
    simulationSpace=BaseBoneSpace
    alphaBoolBlend {blendOption=Linear}
    baseBoneRef {boneName=root}
    componentAppliedLinearAccClamp=<3000 3000 3000>
    componentLinearAccScale=<0.4 0.4 0.42>
    componentLinearVelScale=<0.4 0.4 0.4>
    overrideWorldGravity {z=-980}
   }
   n102 = "Transform (Modify) Bone - Bone: dyn_tie_02"(ComponentPose=@n101, Alpha=1, Translation="0,0,0", Rotation="-5,0,0", Scale="1,1,1") {
    rotationMode=BMM_Additive
    rotationSpace=BCS_ParentBoneSpace
    alphaBoolBlend {blendOption=Linear}
    boneToModify {boneName=dyn_tie_02}
    rotation {pitch=-5}
   }
   n103 = "Constraint - dyn_armband_knot_MSTR (PPP)"!(ComponentPose=@n102, Alpha=1, ConstraintWeights_0=0.541, ConstraintWeights_1:float=0.228, ConstraintWeights_2:float=0.231) -> (Pose) {
    alphaBoolBlend {blendOption=Linear}
    boneToModify {boneName=dyn_armband_knot_MSTR}
    constraintSetup [{targetBone {boneName=upperarm_twist_02_r}}, {targetBone {boneName=upperarm_r}}, {targetBone {boneName=upperarm_twist_02_r}}]
    constraintWeights [1,1,1]
   }
   n104 = "Constraint - master_collar_back_l (PPPPPPPP)"!(
    ComponentPose=@n103
    Alpha=1
    ConstraintWeights_0=0.3022
    ConstraintWeights_1:float=0.0081
    ConstraintWeights_2:float=0.3981
    ConstraintWeights_3:float=0.0011
    ConstraintWeights_4:float=0.0121
    ConstraintWeights_5:float=0.2186
    ConstraintWeights_6:float=0.0542
    ConstraintWeights_7:float=0.0056
   ) -> (Pose) {
    alphaBoolBlend {blendOption=Linear}
    boneToModify {boneName=master_collar_back_l}
    constraintSetup [
     {targetBone {boneName=clavicle_l}}
     {targetBone {boneName=dyn_collar_back_r}}
     {targetBone {boneName=dyn_collar_back_l}}
     {targetBone {boneName=spine_03}}
     {targetBone {boneName=spine_04}}
     {targetBone {boneName=spine_05}}
     {targetBone {boneName=upperarm_twist_01_l}}
     {targetBone {boneName=clavicle_r}}
    ]
    constraintWeights [1,1,1,1,1,1,1,1]
   }
   n105 = "Constraint - master_collar_back_r (PPPPPPPP)"!(
    ComponentPose=@n104
    Alpha=1
    ConstraintWeights_0=0.0056
    ConstraintWeights_1:float=0.056
    ConstraintWeights_2:float=0.3996
    ConstraintWeights_3:float=0.0075
    ConstraintWeights_4:float=0.0012
    ConstraintWeights_5:float=0.0126
    ConstraintWeights_6:float=0.213
    ConstraintWeights_7:float=0.3046
   ) -> (Pose) {
    alphaBoolBlend {blendOption=Linear}
    boneToModify {boneName=master_collar_back_r}
    constraintSetup [
     {targetBone {boneName=clavicle_l}}
     {targetBone {boneName=upperarm_twist_01_r}}
     {targetBone {boneName=dyn_collar_back_r}}
     {targetBone {boneName=dyn_collar_back_l}}
     {targetBone {boneName=spine_03}}
     {targetBone {boneName=spine_04}}
     {targetBone {boneName=spine_05}}
     {targetBone {boneName=clavicle_r}}
    ]
    constraintWeights [1,1,1,1,1,1,1,1]
   }
   n106 = "Transform (Modify) Bone - Bone: dyn_skirt_01_r_MSTR"(ComponentPose=@n105, Alpha=1, Translation="0,0,2", Rotation="0,0,0", Scale="1,1,1") {
    rotationMode=BMM_Additive
    rotationSpace=BCS_BoneSpace
    scaleMode=BMM_Additive
    scaleSpace=BCS_BoneSpace
    translationMode=BMM_Additive
    translationSpace=BCS_BoneSpace
    alphaBoolBlend {blendOption=Linear}
    boneToModify {boneName=dyn_skirt_01_r_MSTR}
    translation {z=2}
   }
   n107 = "Constraint - master_collar_back_l (PPPPPPPP)"(
    ComponentPose=@n106
    Alpha=1
    ConstraintWeights_0=0.5259
    ConstraintWeights_1=0.0055
    ConstraintWeights_2=0.0039
    ConstraintWeights_3=0.0072
    ConstraintWeights_4=0.0624
    ConstraintWeights_5=0.2584
    ConstraintWeights_6=0.1074
    ConstraintWeights_7=0.0293
   ) {
    alphaBoolBlend {blendOption=Linear}
    boneToModify {boneName=master_collar_back_l}
    constraintSetup [
     {targetBone {boneName=clavicle_l}}
     {targetBone {boneName=upperarm_twist_01_r}}
     {targetBone {boneName=spine_02}}
     {targetBone {boneName=spine_03}}
     {targetBone {boneName=spine_04}}
     {targetBone {boneName=spine_05}}
     {targetBone {boneName=upperarm_twist_01_l}}
     {targetBone {boneName=clavicle_r}}
    ]
    constraintWeights [1,1,1,1,1,1,1,1]
   }
   n108 = "Constraint - master_collar_back_r (PPPPPPPPPP)"!(
    ComponentPose=@n107
    Alpha=1
    ConstraintWeights_0=0.0184
    ConstraintWeights_1:float=0.0771
    ConstraintWeights_2:float=0.1716
    ConstraintWeights_3:float=0
    ConstraintWeights_4:float=0.003
    ConstraintWeights_5:float=0.0053
    ConstraintWeights_6:float=0.0473
    ConstraintWeights_7:float=0.2098
    ConstraintWeights_8:float=0.0011
    ConstraintWeights_9:float=0.4664
   ) -> (Pose) {
    alphaBoolBlend {blendOption=Linear}
    boneToModify {boneName=master_collar_back_r}
    constraintSetup [
     {targetBone {boneName=clavicle_l}}
     {targetBone {boneName=upperarm_twist_01_r}}
     {targetBone {boneName=dyn_collar_back_r}}
     {targetBone {boneName=dyn_collar_back_l}}
     {targetBone {boneName=spine_02}}
     {targetBone {boneName=spine_03}}
     {targetBone {boneName=spine_04}}
     {targetBone {boneName=spine_05}}
     {targetBone {boneName=upperarm_twist_01_l}}
     {targetBone {boneName=clavicle_r}}
    ]
    constraintWeights [1,1,1,1,1,1,1,1,1,1]
   }
   n109 = "Transform (Modify) Bone - Bone: master_collar_back_l"(ComponentPose=@n108, Alpha=1, Translation="0,-1.5,0", Rotation="0.000,0.000,3", Scale="1,1,1") {
    rotationMode=BMM_Additive
    scaleMode=BMM_Additive
    translationMode=BMM_Additive
    alphaBoolBlend {blendOption=Linear}
    boneToModify {boneName=master_collar_back_l}
    rotation {roll=3}
    translation {y=-1.5}
   }
   n110 = "Transform (Modify) Bone - Bone: master_collar_back_r"(ComponentPose=@n109, Alpha=1, Translation="0,-1.5,0", Rotation="0,0,3", Scale="1,1,1") {
    rotationMode=BMM_Additive
    scaleMode=BMM_Additive
    translationMode=BMM_Additive
    alphaBoolBlend {blendOption=Linear}
    boneToModify {boneName=master_collar_back_r}
    rotation {roll=3}
    translation {y=-1.5}
   }
   n111 = Get BackblingAlpha() -> (BackblingAlpha:float)
   n112 = Anim Dynamics - Bone: dyn_collar_back_l(ComponentPose=@n110, Alpha=@n111, ExternalForce="0,0,0") {
    angularDampingOverride=0.9999
    angularSpringConstant=75
    bAngularSpring=true
    bGravityOverrideInSimSpace=true
    bOverrideAngularDamping=true
    gravityScale=0.8
    simulationSpace=RootRelative
    alphaBoolBlend {blendOption=Linear}
    boundBone {boneName=dyn_collar_back_l}
    componentAppliedLinearAccClamp=<5000 5000 5000>
    componentLinearAccScale=<0.85 0.85 0.85>
    componentLinearVelScale=<0.85 0.85 0.85>
    physicsBodyDefinitions [{
      sphereCollisionRadius=0
      boundBone {boneName=dyn_collar_back_l}
      boxExtents=<20 20 10>
      constraintSetup {angularLimitsMax=<15 15 0>, angularLimitsMin=<-15 -15 -90>, angularTarget=<0 0 0>}
      localJointOffset=<2 0 0>
     }]
   }
   n113 = Anim Dynamics - Bone: dyn_collar_back_r(ComponentPose=@n112, Alpha=@n111, ExternalForce="0,0,0") {
    angularDampingOverride=0.99
    angularSpringConstant=75
    bAngularSpring=true
    bGravityOverrideInSimSpace=true
    bOverrideAngularDamping=true
    gravityScale=0.8
    simulationSpace=RootRelative
    alphaBoolBlend {blendOption=Linear}
    boundBone {boneName=dyn_collar_back_r}
    componentAppliedLinearAccClamp=<5000 5000 5000>
    componentLinearAccScale=<0.85 0.85 0.85>
    componentLinearVelScale=<0.85 0.85 0.85>
    physicsBodyDefinitions [{
      sphereCollisionRadius=0
      boundBone {boneName=dyn_collar_back_r}
      boxExtents=<20 20 10>
      constraintSetup {angularLimitsMax=<15 15 0>, angularLimitsMin=<-15 -15 -90>, angularTarget=<0 0 0>}
      localJointOffset=<2 0 0>
     }]
   }
   n114 = "Transform (Modify) Bone - Bone: master_collar_back_r"!(ComponentPose=@n113, bAlphaBoolEnabled:bool=Get bIsPlayingMeleeAnim(), Translation="0,-3,0", Rotation="0,0,0", Scale="1,1,1") -> (Pose) {
    alphaInputType=Bool
    rotationMode=BMM_Additive
    scaleMode=BMM_Additive
    translationMode=BMM_Additive
    alphaBoolBlend {blendInTime=0.2, blendOption=Linear, blendOutTime=0.1}
    boneToModify {boneName=master_collar_back_r}
    translation {y=-3}
   }
   n115 = Get bSkydiving() -> (bSkydiving:bool)
   n116 = "Transform (Modify) Bone - Bone: master_collar_back_l"!(ComponentPose=@n114, bAlphaBoolEnabled:bool=@n115, Translation="0,0,2", Rotation="0.000,0.000,10", Scale="1,1,1") -> (Pose) {
    alphaInputType=Bool
    rotationMode=BMM_Additive
    scaleMode=BMM_Additive
    translationMode=BMM_Additive
    alphaBoolBlend {blendInTime=0.1, blendOption=Linear, blendOutTime=0.1}
    boneToModify {boneName=master_collar_back_l}
    rotation {roll=10}
    translation {z=2}
   }
   n117 = "Transform (Modify) Bone - Bone: master_collar_back_r"(ComponentPose=@n116, bAlphaBoolEnabled=@n115, Translation="0,0,2", Rotation="0,0,10", Scale="1,1,1") {
    alphaInputType=Bool
    rotationMode=BMM_Additive
    scaleMode=BMM_Additive
    translationMode=BMM_Additive
    alphaBoolBlend {blendInTime=0.1, blendOption=Linear, blendOutTime=0.1}
    boneToModify {boneName=master_collar_back_r}
    rotation {roll=10}
    translation {z=2}
   }
   n118 = Get bIsSkydiveDiveMode()
   n119 = "Transform (Modify) Bone - Bone: master_collar_back_l"(ComponentPose=@n117, bAlphaBoolEnabled=@n118, Translation="0,0,2", Rotation="0.000,0.000,10", Scale="1,1,1") {
    alphaInputType=Bool
    rotationMode=BMM_Additive
    scaleMode=BMM_Additive
    translationMode=BMM_Additive
    alphaBoolBlend {blendInTime=0.1, blendOption=Linear, blendOutTime=0.1}
    boneToModify {boneName=master_collar_back_l}
    rotation {roll=10}
    translation {z=2}
   }
   n120 = "Transform (Modify) Bone - Bone: master_collar_back_r"(ComponentPose=@n119, bAlphaBoolEnabled=@n118, Translation="0,0,2", Rotation="0,0,10", Scale="1,1,1") {
    alphaInputType=Bool
    rotationMode=BMM_Additive
    scaleMode=BMM_Additive
    translationMode=BMM_Additive
    alphaBoolBlend {blendInTime=0.1, blendOption=Linear, blendOutTime=0.1}
    boneToModify {boneName=master_collar_back_r}
    rotation {roll=10}
    translation {z=2}
   }
   n121 = RigidBody!(
    ComponentPose=@n120
    Alpha=Get FrontendSkirtAlpha() -> (FrontendSkirtAlpha:float)
    OverrideWorldGravity:Vector=Get Gravity_Override_Skirt()
    ExternalForce
   ) -> (Pose) {
    bClampLinearTranslationLimitToRefPose=true
    bOverrideWorldGravity=true
    overridePhysicsAsset=&/F_MED_Cyber_Karate_Human_Dress_FE_Physics
    simulationSpace=BaseBoneSpace
    alphaBoolBlend {blendOption=Linear}
    baseBoneRef {boneName=root}
    componentAppliedLinearAccClamp=<1000 1000 1000>
    componentLinearAccScale=<0.25 0.25 0.3>
    overrideWorldGravity {z=-980}
   }
   n122 = Get FrontendNoEmote() -> (FrontendNoEmote:float)
   n123 = "Transform (Modify) Bone - Bone: dyn_skirt_01_l_MSTR"(ComponentPose=@n121, Alpha=@n122, Translation="0,0,0", Rotation="-10,0,0", Scale="1,1,1") {
    rotationMode=BMM_Additive
    translationMode=BMM_Additive
    alphaBoolBlend {blendOption=Linear}
    boneToModify {boneName=dyn_skirt_01_l_MSTR}
    rotation {pitch=-10}
   }
   n124 = "Transform (Modify) Bone - Bone: dyn_skirt_01_r_MSTR"(ComponentPose=@n123, Alpha=@n122, Translation="0,0,0", Rotation="15,0,0", Scale="1,1,1") {
    rotationMode=BMM_Additive
    translationMode=BMM_Additive
    alphaBoolBlend {blendOption=Linear}
    boneToModify {boneName=dyn_skirt_01_r_MSTR}
    rotation {pitch=15}
   }
   n125 = "Transform (Modify) Bone - Bone: dyn_skirt_fr2_01_l_MSTR"(ComponentPose=@n124, Alpha=@n122, Translation="0,0,0", Rotation="-5,0,0", Scale="1,1,1") {
    rotationMode=BMM_Additive
    translationMode=BMM_Additive
    alphaBoolBlend {blendOption=Linear}
    boneToModify {boneName=dyn_skirt_fr2_01_l_MSTR}
    rotation {pitch=-5}
   }
   n126 = "Transform (Modify) Bone - Bone: dyn_skirt_fr2_01_r_MSTR"(ComponentPose=@n125, Alpha=@n122, Translation="0,0,0", Rotation="10,0,0", Scale="1,1,1") {
    rotationMode=BMM_Additive
    translationMode=BMM_Additive
    alphaBoolBlend {blendOption=Linear}
    boneToModify {boneName=dyn_skirt_fr2_01_r_MSTR}
    rotation {pitch=10}
   }
   Output Pose(Result=Component To Local(ComponentPose=@n126)) {layerGroup=None}
   n127 = "Constraint - dyn_thigh_01_l (RR)"!(
    ComponentPose=Local To Component(LocalPose=Copy Pose From Mesh() {bUseAttachedParent=true})
    Alpha=1
    ConstraintWeights_0=1
    ConstraintWeights_1:float=0
   ) -> (Pose) {
    alphaBoolBlend {blendOption=Linear}
    boneToModify {boneName=dyn_thigh_01_l}
    constraintSetup [{transformType=Rotation, targetBone {boneName=thigh_l}}, {transformType=Rotation, targetBone {boneName=pelvis}}]
    constraintWeights [1,1]
   }
   n128 = "Constraint - dyn_thigh_01_r (RR)"!(ComponentPose=@n127, Alpha=1, ConstraintWeights_0=1, ConstraintWeights_1:float=0) -> (Pose) {
    alphaBoolBlend {blendOption=Linear}
    boneToModify {boneName=dyn_thigh_01_r}
    constraintSetup [{transformType=Rotation, targetBone {boneName=thigh_r}}, {transformType=Rotation, targetBone {boneName=pelvis}}]
    constraintWeights [1,1]
   }
   n129 = "Constraint - dyn_skirt_fr_01_MSTR (RR)"!(ComponentPose=@n128, Alpha=1, ConstraintWeights_0=0.5, ConstraintWeights_1:float=0.5) -> (Pose) {
    alphaBoolBlend {blendOption=Linear}
    boneToModify {boneName=dyn_skirt_fr_01_MSTR}
    constraintSetup [
     {transformType=Rotation, targetBone {boneName=dyn_skirt_fr_01_l_MSTR}}
     {transformType=Rotation, targetBone {boneName=dyn_skirt_fr_01_r_MSTR}}
    ]
    constraintWeights [1,1]
   }
   n130 = "Constraint - dyn_skirt_fr_01 (RR)"!(ComponentPose=@n129, Alpha=1, ConstraintWeights_0=0.5, ConstraintWeights_1:float=0.5) -> (Pose) {
    alphaBoolBlend {blendOption=Linear}
    boneToModify {boneName=dyn_skirt_fr_01}
    constraintSetup [{transformType=Rotation, targetBone {boneName=dyn_skirt_fr_01_l}}, {transformType=Rotation, targetBone {boneName=dyn_skirt_fr_01_r}}]
    constraintWeights [1,1]
   }
   n131 = "Constraint - dyn_skirt_fr_02 (RR)"!(ComponentPose=@n130, Alpha=1, ConstraintWeights_0=0.5, ConstraintWeights_1:float=0.5) -> (Pose) {
    alphaBoolBlend {blendOption=Linear}
    boneToModify {boneName=dyn_skirt_fr_02}
    constraintSetup [{transformType=Rotation, targetBone {boneName=dyn_skirt_fr_02_l}}, {transformType=Rotation, targetBone {boneName=dyn_skirt_fr_02_r}}]
    constraintWeights [1,1]
   }
   n132 = "Constraint - dyn_skirt_fr_03 (RR)"!(ComponentPose=@n131, Alpha=1, ConstraintWeights_0=0.5, ConstraintWeights_1:float=0.5) -> (Pose) {
    alphaBoolBlend {blendOption=Linear}
    boneToModify {boneName=dyn_skirt_fr_03}
    constraintSetup [{transformType=Rotation, targetBone {boneName=dyn_skirt_fr_03_l}}, {transformType=Rotation, targetBone {boneName=dyn_skirt_fr_03_r}}]
    constraintWeights [1,1]
   }
   Save cached pose 'PreDressRBAN'(Pose=Component To Local(ComponentPose=@n132)) {cacheName=PreDressRBAN}
  ]}
 EventGraph {nodes [
   Event BlueprintUpdateAnimation()
   > Gravity Override State Function DEPRECATED(CharacterPartAnimInstance=Self-Reference(), GameTimeInSeconds=GetGameTimeInSeconds(), GravityOverrideParams=Get Gravity Override Params())
   > Set Gravity_Override_Skirt(Gravity_Override_Skirt:Vector="0, 0, 0") -> (Output_Get:Vector)
  ]}
]
variables [
 Gravity Override Params {type=GravityOverrideParamsStruct}
 Gravity_Override_Skirt {type=Vector}
]
