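# Physics asset for the FE mesh part of the female medium Cyber Karate Human cosmetic, covering its hair bun and loose strands.
# Capsule bodies sit on upper spine, neck, head and dyn_ hair bones, linked by free-rotation joints up to the hair base and bun master.
# Single strands, two chains and the bun hang from these with angular limits: the left chain is nearly rigid past its root, the right swings widely with little twist.
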PhysicsAsset /BRCosmetics/Characters/Player/Female/Medium/Bodies/F_MED_Cyber_Karate_Human/Meshes/Parts/F_MED_Cyber_Karate_Human_FE_Physics
bodies {
 dyn_bun [{shapeType=Capsule, length=5.0575, radius=2.0821, center=<-1.1884 0 0>}]
 dyn_bun_MSTR [{shapeType=Capsule, length=0.505, radius=0.505, rotation=<rot 0 -90 0>}]
 dyn_bun_strand [{shapeType=Capsule, length=2.9898, radius=2.3213, center=<-1.1884 0 0>}]
 dyn_hair_base [{shapeType=Capsule, length=0.505, radius=0.505, rotation=<rot 0 -90 0>}]
 dyn_hair_l_00 [{shapeType=Capsule, length=5.1122, radius=1.5687, center=<4.0824 0 -0.046>, rotation=<rot 90 0 0>}]
 dyn_hair_l_01 [{shapeType=Capsule, length=1, radius=1, center=<1.4185 0 0>, rotation=<rot 90 0 0>}]
 dyn_hair_l_02 [{shapeType=Capsule, length=1, radius=1, center=<1.4828 0 0>, rotation=<rot 90 0 0>}]
 dyn_hair_l_03 [{shapeType=Capsule, length=1, radius=1, center=<1.4265 0 0>, rotation=<rot 90 0 0>}]
 dyn_hair_l_04 [{shapeType=Capsule, length=1, radius=1, center=<1.1766 0 0>, rotation=<rot 90 0 0>}]
 dyn_hair_l_05 [{shapeType=Capsule, length=1, radius=1, center=<1.4569 0 1>, rotation=<rot 90 0 0>}]
 dyn_hair_l_07 [{shapeType=Capsule, length=1, radius=1, center=<1.4569 0 1>, rotation=<rot 90 0 0>}]
 dyn_hair_l_08 [{shapeType=Capsule, length=1, radius=1, center=<1.4569 0 1>, rotation=<rot 90 0 0>}]
 dyn_hair_l_09 [{shapeType=Capsule, length=1, radius=1, center=<1.1766 0 0>, rotation=<rot 90 0 0>}]
 dyn_hair_l_10 [{shapeType=Capsule, length=1, radius=1, center=<1.1617 -1 0>, rotation=<rot 90 0 0>}]
 dyn_hair_r_01 [{shapeType=Capsule, length=2, radius=2, center=<1.9863 0 0>}]
 dyn_hair_r_02 [{shapeType=Capsule, length=2, radius=2, center=<1.6856 0 0>}]
 dyn_hair_r_03 [{shapeType=Capsule, length=2, radius=2, center=<1.8828 0 -0.0001>}]
 dyn_hair_r_05 [{shapeType=Capsule, length=1, radius=1, center=<1.4611 0 0>, rotation=<rot 90 0 0>}]
 dyn_hair_r_06 [{shapeType=Capsule, length=1, radius=1, rotation=<rot 90 0 0>}]
 head [{shapeType=Capsule, length=3.6124, radius=6.2206, center=<2.8691 -0.0001 0>, rotation=<rot 90 0 0>}]
 neck_01 [{shapeType=Capsule, length=4, radius=4, rotation=<rot 0 -90 0>}]
 neck_02 [{shapeType=Capsule, length=4, radius=4, rotation=<rot 0 -90 0>}]
 spine_05 [{shapeType=Capsule, length=46.7298, radius=11.4743, center=<0 -2.1617 0>, rotation=<rot 0 -90 0>}]
}
constraints [
 {
  bone1Name=dyn_hair_base
  bone2Name=head
  swing1LimitDegrees=45
  swing1Motion=Free
  swing2LimitDegrees=45
  swing2Motion=Free
  twistLimitDegrees=45
  twistMotion=Free
 }
 {
  bone1Name=dyn_bun_MSTR
  bone2Name=dyn_hair_base
  swing1LimitDegrees=45
  swing1Motion=Free
  swing2LimitDegrees=45
  swing2Motion=Free
  twistLimitDegrees=45
  twistMotion=Free
 }
 {bone1Name=dyn_bun, bone2Name=dyn_bun_MSTR, swing1LimitDegrees=45, swing2LimitDegrees=45, twistLimitDegrees=25}
 {bone1Name=dyn_bun_strand, bone2Name=dyn_bun, swing1LimitDegrees=45, swing2LimitDegrees=45, twistLimitDegrees=25}
 {bone1Name=dyn_hair_r_05, bone2Name=dyn_hair_base, swing1LimitDegrees=45, swing2LimitDegrees=45, twistLimitDegrees=45}
 {bone1Name=dyn_hair_l_04, bone2Name=dyn_hair_base, swing1LimitDegrees=45, swing2LimitDegrees=45, twistLimitDegrees=45}
 {bone1Name=dyn_hair_l_05, bone2Name=dyn_hair_base, swing1LimitDegrees=45, swing2LimitDegrees=45, twistLimitDegrees=45}
 {bone1Name=dyn_hair_r_06, bone2Name=dyn_hair_base, swing1LimitDegrees=45, swing2LimitDegrees=45, twistLimitDegrees=45}
 {bone1Name=dyn_hair_l_07, bone2Name=dyn_hair_base, swing1LimitDegrees=45, swing2LimitDegrees=45, twistLimitDegrees=45}
 {bone1Name=dyn_hair_l_08, bone2Name=dyn_hair_base, swing1LimitDegrees=45, swing2LimitDegrees=45, twistLimitDegrees=45}
 {bone1Name=dyn_hair_l_09, bone2Name=dyn_hair_base, swing1LimitDegrees=45, swing2LimitDegrees=45, twistLimitDegrees=45}
 {bone1Name=dyn_hair_l_10, bone2Name=dyn_hair_base, swing1LimitDegrees=45, swing2LimitDegrees=45, twistLimitDegrees=45}
 {bone1Name=dyn_hair_l_00, bone2Name=dyn_hair_base, swing1LimitDegrees=45, swing2LimitDegrees=45, twistLimitDegrees=45}
 {bone1Name=dyn_hair_l_01, bone2Name=dyn_hair_l_00, swing1LimitDegrees=1, swing2LimitDegrees=1, twistLimitDegrees=1}
 {bone1Name=dyn_hair_l_02, bone2Name=dyn_hair_l_01, swing1LimitDegrees=1, swing2LimitDegrees=1, twistLimitDegrees=1}
 {bone1Name=dyn_hair_l_03, bone2Name=dyn_hair_l_02, swing1LimitDegrees=1, swing2LimitDegrees=1, twistLimitDegrees=1}
 {bone1Name=dyn_hair_r_01, bone2Name=dyn_hair_base, swing1LimitDegrees=50, swing2LimitDegrees=50, twistLimitDegrees=2}
 {bone1Name=dyn_hair_r_02, bone2Name=dyn_hair_r_01, swing1LimitDegrees=50, swing2LimitDegrees=50, twistLimitDegrees=2}
 {bone1Name=dyn_hair_r_03, bone2Name=dyn_hair_r_02, swing1LimitDegrees=50, swing2LimitDegrees=50, twistLimitDegrees=2}
 {
  bone1Name=head
  bone2Name=neck_02
  swing1LimitDegrees=45
  swing1Motion=Free
  swing2LimitDegrees=45
  swing2Motion=Free
  twistLimitDegrees=45
  twistMotion=Free
 }
 {
  bone1Name=neck_02
  bone2Name=neck_01
  swing1LimitDegrees=45
  swing1Motion=Free
  swing2LimitDegrees=45
  swing2Motion=Free
  twistLimitDegrees=45
  twistMotion=Free
 }
 {
  bone1Name=neck_01
  bone2Name=spine_05
  swing1LimitDegrees=45
  swing1Motion=Free
  swing2LimitDegrees=45
  swing2Motion=Free
  twistLimitDegrees=45
  twistMotion=Free
 }
]
