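# Physics asset for the hair bun of the female medium Cyber Karate Human player character, covering the head and dynamic hair bones.
# Five single-capsule bodies form a constraint chain from head through hair base, bun master and bun to the bun strand.
# The two joints nearest the head have free swing and twist, while the bun and strand joints carry swing limits with a much tighter twist limit.
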PhysicsAsset /BRCosmetics/Characters/Player/Female/Medium/Bodies/F_MED_Cyber_Karate_Human/Meshes/Parts/F_MED_Cyber_Karate_Human_Physics
bodies {
 dyn_bun [{shapeType=Capsule, length=5.0401, radius=3.4321, center=<-1.1884 0 0>}]
 dyn_bun_MSTR [{shapeType=Capsule, length=0.505, radius=0.505, rotation=<rot 0 -90 0>}]
 dyn_bun_strand [{shapeType=Capsule, length=5.0401, radius=3.4321, center=<-1.1884 0 0>}]
 dyn_hair_base [{shapeType=Capsule, length=0.505, radius=0.505, rotation=<rot 0 -90 0>}]
 head [{shapeType=Capsule, length=1, radius=1}]
}
constraints [
 {
  bone1Name=dyn_hair_base
  bone2Name=head
  swing1LimitDegrees=45
  swing1Motion=Free
  swing2LimitDegrees=45
  swing2Motion=Free
  twistLimitDegrees=45
  twistMotion=Free
 }
 {
  bone1Name=dyn_bun_MSTR
  bone2Name=dyn_hair_base
  swing1LimitDegrees=45
  swing1Motion=Free
  swing2LimitDegrees=45
  swing2Motion=Free
  twistLimitDegrees=45
  twistMotion=Free
 }
 {bone1Name=dyn_bun, bone2Name=dyn_bun_MSTR, swing1LimitDegrees=45, swing2LimitDegrees=45, twistLimitDegrees=5}
 {bone1Name=dyn_bun_strand, bone2Name=dyn_bun, swing1LimitDegrees=45, swing2LimitDegrees=45, twistLimitDegrees=5}
]
